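# Material function for UGC UI materials that outputs a selected time style from an index: static value, loop, ping-pong or switch.
# MF_UGCUI_LinearTime supplies the three animated 0-1 signals, with a Speed input exposed alongside the static value.
# MF_UI_CheckIntEquality compares the index to 0-3, and each result drives a chain of lerps as a branchless selector.
# The innermost lerp has identical A and B, so unlisted indices behave like No Time.
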MaterialFunction /FortUILibrary/UGC/MaterialFunctions/MF_UGCUI_TimeSwitcher
graph {
 properties {description=Select the Time method you want to use, libraryCategoriesText [""]}
 nodes [
  n1 = NamedRerouteUsage()
  n2 = MaterialFunctionCall!(Int A=NamedRerouteUsage(), Int B=Constant() {desc="No time used (0)"}) {materialFunction=&/MF_UI_CheckIntEquality}
  n3 = NamedRerouteUsage()
  n4 = Constant() {desc="Loop (1)", r=1}
  n5 = LinearInterpolate(
   A=LinearInterpolate(A=@n1, B=@n1, Alpha=NamedRerouteDeclaration(None=@n2.Result) {name=is No Time, desc="0 = No, 1 = Yes", nodeColor=<rgba 0 0 0 1>}) {desc="A = Any Integer not listed"}
   B=MaterialFunctionCall!(FloatValue, Speed=NamedRerouteUsage()) {materialFunction=&/MF_UGCUI_LinearTime}.0-1 Loop
   Alpha=NamedRerouteDeclaration(None=MaterialFunctionCall(Int A=@n3, Int B=@n4) {materialFunction=&MF_UI_CheckIntEquality}.Result) {name=isLoop, desc="0 = No, 1 = Yes", nodeColor=<rgba 0 0 0 1>}
  )
  n6 = MaterialFunctionCall(Speed=NamedRerouteUsage()) {materialFunction=&MF_UGCUI_LinearTime}
  n7 = NamedRerouteDeclaration(None=MaterialFunctionCall(Int A=NamedRerouteUsage(), Int B=Constant() {desc="Ping Pong (2)", r=2}) {materialFunction=&MF_UI_CheckIntEquality}.Result) {name=is Ping Pong, desc="0 = No, 1 = Yes", nodeColor=<rgba 0 0 0 1>}
  n8 = NamedRerouteUsage()
  n9 = MaterialFunctionCall(Int A=NamedRerouteUsage(), Int B=Constant() {desc="Switch (3)", r=3}) {materialFunction=&MF_UI_CheckIntEquality}
  FunctionOutput(None=LinearInterpolate(
    A=LinearInterpolate(A=@n5, B=@n6.0-1 Ping pong, Alpha=@n7)
    B=MaterialFunctionCall(Speed=@n8) {materialFunction=&MF_UGCUI_LinearTime}.0-1 Switch
    Alpha=NamedRerouteDeclaration(None=@n9.Result) {name=is Switch, desc="0 = No, 1 = Yes", nodeColor=<rgba 0 0 0 1>}
   )) {outputName=Selected Time Style}
  n10 = FunctionInput() {
   bUsePreviewValueAsDefault=true
   description=The static value used if you decide to not interpolate through the time.
   inputName=Speed
   inputType=FunctionInput_Scalar
   previewValue=<0.2 0 0 0>
  }
  NamedRerouteDeclaration(None=@n10) {name=Time Speed, nodeColor=<rgba 0 0 0 1>}
  n11 = FunctionInput() {
   bUsePreviewValueAsDefault=true
   description=The static value used if you decide to not interpolate through the time.
   inputName=No Time Static Value
   inputType=FunctionInput_Scalar
   previewValue=<0 0 0 0>
  }
  NamedRerouteDeclaration(None=@n11) {name=Static Value, nodeColor=<rgba 0 0 0 1>}
  n12 = FunctionInput(Preview=Constant() {desc="Time Style Switch: 0 = No Time used (default), 1 = Loop, 2 = Ping Pong, 3 = Switch"}) {
   bUsePreviewValueAsDefault=true
   description="Time Style Switch (from 0 to 1): 0 = No Time used, 1 = Loop, 2 = Ping Pong, 3 = Switch"
   inputName=Time Style Switch Index
   inputType=FunctionInput_Scalar
   sortPriority=1
  }
  NamedRerouteDeclaration(None=@n12) {name=Time Style Switch, nodeColor=<rgba 1 0 0.4833 1>}
 ]
}
