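# Translucent UI material that draws a configurable box with per-corner roundness, stroke, gap and optional linear or radial gradient for UGC widgets.
# MF_UGCUI_Box builds masks from 9-slice or normalized UVs, with slant, rotation and scaling helpers.
# Gradient rotation can animate through time and easing switcher functions.
# Idle, Hovered, Focused, Pressed and Disabled each compute color and alpha, blended by chained lerps on Trans_* scalars in that order.
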Material /FortUILibrary/UGC/Materials/M_Basic_Box
graph {
 properties {blendMode=BLEND_Translucent, materialDomain=MD_UI, pixelDepthOffsetMode=PDOM_Legacy, usedShadingModels=Default Lit}
 nodes [
  n1 = ScalarParameter() {
   defaultValue=1
   desc="Default Value = 1.0\r\n\r\nSet the Width of your shape from 0 to 1 (100%)\r\n"
   group="[1] Fill"
   parameterName=Width
   sliderMax=1
   sortPriority=5
  }
  n2 = NamedRerouteUsage()
  n3 = ScalarParameter() {
   defaultValue=1
   desc="Default Value = 1.0\r\n\r\nSet the Height of your shape from 0 to 1 (100%) \r\n"
   group="[1] Fill"
   parameterName=Height
   sliderMax=1
   sortPriority=6
  }
  NamedRerouteDeclaration(None=AppendVector(A=Multiply(A=@n1, B=@n2), B=Multiply(A=@n2, B=@n3))) {name=Size, nodeColor=<rgba 0 0 0 1>}
  n4 = ScalarParameter() {
   desc="Default Value = 0.0\r\n\r\nSet the rounded corners of the shape."
   group="[1] Fill"
   parameterName=Top Left Corner Roundness
   sliderMax=0.5
   sortPriority=8
  }
  n5 = ScalarParameter() {
   desc="Default Value = 0.0\r\n\r\nSet the rounded corners of the shape."
   group="[1] Fill"
   parameterName=Bottom Left Corner Roundness
   sliderMax=0.5
   sortPriority=11
  }
  n6 = ScalarParameter() {
   desc="Default Value = 0.0\r\n\r\nSet the rounded corners of the shape."
   group="[1] Fill"
   parameterName=Top Right Corner Roundness
   sliderMax=0.5
   sortPriority=9
  }
  n7 = ScalarParameter() {
   desc="Default Value = 0.0\r\n\r\nSet the rounded corners of the shape."
   group="[1] Fill"
   parameterName=Bottom Right Corner Roundness
   sliderMax=0.5
   sortPriority=10
  }
  NamedRerouteDeclaration(None=MaterialFunctionCall(X=@n4, Y=@n5, Z=@n6, A=@n7) {materialFunction=&MakeFloat4}.Result) {name=Corners Roundness, nodeColor=<rgba 0 0 0 1>}
  n8 = ScalarParameter() {
   desc="Default Value = 0 (0-1)\r\n\r\nSet the transparency between 0 (transparent) and 1 (full opaque).\r\n"
   group="[4d] State-Disabled"
   parameterName=Disabled Stroke Transparency
   sliderMax=1
   sortPriority=6
  }
  NamedRerouteDeclaration(None=@n8) {name=Disabled Stroke Transparency, nodeColor=<rgba 0 0 0 1>}
  n9 = NamedRerouteUsage()
  n10 = NamedRerouteUsage()
  n11 = NamedRerouteUsage()
  n12 = Saturate(None=Multiply(A=Multiply(A=@n9, B=@n10), B=ComponentMask(None=@n11) {a=true, desc="The alpha of the Shape Color \r\nalso drives the shape alpha"}))
  n13 = NamedRerouteUsage()
  n14 = NamedRerouteUsage()
  n15 = NamedRerouteUsage()
  n16 = Saturate(None=Multiply(A=Multiply(A=@n13, B=@n14), B=ComponentMask(None=@n15) {a=true, desc="The alpha of the stroke color \r\nalso drives the stroke alpha"}))
  NamedRerouteDeclaration(None=Max(
    A=NamedRerouteDeclaration(None=@n12) {name=Hovered Shape Alpha, nodeColor=<rgba 0 0 0 1>}
    B=NamedRerouteDeclaration(None=@n16) {name=Hovered Stroke Alpha, nodeColor=<rgba 0 0 0 1>}
   )) {name=Hovered Alpha, nodeColor=<rgba 0 0 0 1>}
  n17 = NamedRerouteUsage()
  n18 = NamedRerouteUsage()
  n19 = NamedRerouteUsage()
  n20 = Saturate(None=Multiply(A=Multiply(A=@n17, B=@n18), B=ComponentMask(None=@n19) {a=true, desc="The alpha of the Shape Color \r\nalso drives the shape alpha"}))
  n21 = NamedRerouteUsage()
  n22 = NamedRerouteUsage()
  n23 = NamedRerouteUsage()
  n24 = Saturate(None=Multiply(A=Multiply(A=@n21, B=@n22), B=ComponentMask(None=@n23) {a=true, desc="The alpha of the stroke color \r\nalso drives the stroke alpha"}))
  NamedRerouteDeclaration(None=Max(
    A=NamedRerouteDeclaration(None=@n20) {name=Focused Shape Alpha, nodeColor=<rgba 0 0 0 1>}
    B=NamedRerouteDeclaration(None=@n24) {name=Focused Stroke Alpha, nodeColor=<rgba 0 0 0 1>}
   )) {name=Focused Alpha, nodeColor=<rgba 0 0 0 1>}
  n25 = NamedRerouteUsage()
  n26 = NamedRerouteUsage()
  n27 = NamedRerouteUsage()
  n28 = Saturate(None=Multiply(A=Multiply(A=@n25, B=@n26), B=ComponentMask(None=@n27) {a=true, desc="The alpha of the Shape Color \r\nalso drives the shape alpha"}))
  n29 = NamedRerouteUsage()
  n30 = NamedRerouteUsage()
  n31 = NamedRerouteUsage()
  n32 = Saturate(None=Multiply(A=Multiply(A=@n29, B=@n30), B=ComponentMask(None=@n31) {a=true, desc="The alpha of the stroke color \r\nalso drives the stroke alpha"}))
  NamedRerouteDeclaration(None=Max(
    A=NamedRerouteDeclaration(None=@n28) {name=Pressed Shape Alpha, nodeColor=<rgba 0 0 0 1>}
    B=NamedRerouteDeclaration(None=@n32) {name=Pressed Stroke Alpha, nodeColor=<rgba 0 0 0 1>}
   )) {name=Pressed Alpha, nodeColor=<rgba 0 0 0 1>}
  n33 = NamedRerouteUsage()
  n34 = NamedRerouteUsage()
  n35 = NamedRerouteUsage()
  n36 = Saturate(None=Multiply(A=Multiply(A=@n33, B=@n34), B=ComponentMask(None=@n35) {a=true, desc="The alpha of the Shape Color \r\nalso drives the shape alpha"}))
  n37 = NamedRerouteUsage()
  n38 = NamedRerouteUsage()
  n39 = NamedRerouteUsage()
  n40 = Saturate(None=Multiply(A=Multiply(A=@n37, B=@n38), B=ComponentMask(None=@n39) {a=true, desc="The alpha of the stroke color \r\nalso drives the stroke alpha"}))
  NamedRerouteDeclaration(None=Max(
    A=NamedRerouteDeclaration(None=@n36) {name=Disabled Shape Alpha, nodeColor=<rgba 0 0 0 1>}
    B=NamedRerouteDeclaration(None=@n40) {name=Disabled Stroke Alpha, nodeColor=<rgba 0 0 0 1>}
   )) {name=Disabled Alpha, nodeColor=<rgba 0 0 0 1>}
  n41 = ScalarParameter() {
   desc="Switch from a state to another: \r\n\r\n0 = IDLE (default)\r\n1 = HOVERED \r\n"
   group="[4a] State-Hovered"
   parameterName=Trans_Hovered
   sliderMax=1
   sortPriority=11
  }
  NamedRerouteDeclaration(None=Saturate(None=@n41)) {name=Transition Hovered, nodeColor=<rgba 0 0 0 1>}
  n42 = ScalarParameter() {
   desc="Switch from a state to another: \r\n\r\n0 = HOVERED\r\n1 = FOCUSED \r\n"
   group="[4b] State-Focused"
   parameterName=Trans_Focused
   sliderMax=1
   sortPriority=11
  }
  NamedRerouteDeclaration(None=Saturate(None=@n42)) {name=Transition Focused, nodeColor=<rgba 0 0 0 1>}
  n43 = ScalarParameter() {
   desc="Switch from a state to another: \r\n\r\n0 = FOCUSED \r\n1 = PRESSED \r\n"
   group="[4c] State-Pressed"
   parameterName=Trans_Pressed
   sliderMax=1
   sortPriority=11
  }
  NamedRerouteDeclaration(None=Saturate(None=@n43)) {name=Transition Pressed, nodeColor=<rgba 0 0 0 1>}
  n44 = LinearInterpolate(
   A=LinearInterpolate(A=NamedRerouteUsage(), B=NamedRerouteUsage(), Alpha=NamedRerouteUsage())
   B=NamedRerouteUsage()
   Alpha=NamedRerouteUsage()
  )
  n45 = NamedRerouteUsage()
  n46 = NamedRerouteUsage()
  NamedRerouteDeclaration(None=LinearInterpolate(A=LinearInterpolate(A=@n44, B=@n45, Alpha=@n46), B=NamedRerouteUsage(), Alpha=NamedRerouteUsage())) {name=Current Color State, nodeColor=<rgba 0 0 0 1>}
  n47 = ScalarParameter() {
   desc="Switch from a state to another: \r\n\r\n0 = PRESSED \r\n1 = DISABLED \r\n"
   group="[4d] State-Disabled"
   parameterName=Trans_Disabled
   sliderMax=1
   sortPriority=11
  }
  NamedRerouteDeclaration(None=Saturate(None=@n47)) {name=Transition Disabled, nodeColor=<rgba 0 0 0 1>}
  n48 = LinearInterpolate(
   A=LinearInterpolate(A=NamedRerouteUsage(), B=NamedRerouteUsage(), Alpha=NamedRerouteUsage())
   B=NamedRerouteUsage()
   Alpha=NamedRerouteUsage()
  )
  n49 = NamedRerouteUsage()
  n50 = NamedRerouteUsage()
  NamedRerouteDeclaration(None=LinearInterpolate(A=LinearInterpolate(A=@n48, B=@n49, Alpha=@n50), B=NamedRerouteUsage(), Alpha=NamedRerouteUsage())) {name=Current Alpha State, nodeColor=<rgba 0 0 0 1>}
  NamedRerouteDeclaration(None=MaterialFunctionCall!(UVs=NamedRerouteUsage(), Slant amount=NamedRerouteUsage()) {materialFunction=/FortUILibrary/UGC/MaterialFunctions/MF_UGCUI_Slant}.Slanted UVs) {name=Shape Slanted UV, nodeColor=<rgba 1 0.0036 0 1>}
  n51 = ScalarParameter() {
   desc="Default Value = 0.0\r\n\r\nDefine shape X-Axis inclination. Value can be positive or negative.\r\n"
   group="[1] Fill"
   parameterName=Slant amount X
   sortPriority=23
  }
  n52 = ScalarParameter() {
   desc="Default Value = 0.0\r\n\r\nDefine shape Y-Axis inclination. Value can be positive or negative.\r\n"
   group="[1] Fill"
   parameterName=Slant amount Y
   sortPriority=23
  }
  NamedRerouteDeclaration(None=AppendVector(A=@n51, B=@n52)) {name=Slant Amount, nodeColor=<rgba 0 0 0 1>}
  n53 = ScalarParameter() {
   desc="0.0 = Use 9-slice UV (default)\r\nGreater than 0.0 = Use Normalized UV \r\n\r\nTips:\r\n\r\n  - 9-Sliced UV: Use them by default. If you apply 9-Slice to your widget, then your material will follow that cut.\r\n\r\n  - Normalized UV: Use them if you want your material don't be affected by 9-Slice, and then, correctly distributed  along the widget. \r\n"
   group="[1] Fill"
   parameterName=Select the UV type
   sliderMax=3
   sliderMin=1
   sortPriority=1
  }
  NamedRerouteDeclaration(None=Saturate(None=Ceil(None=@n53))) {name=Common UV Type, nodeColor=<rgba 0 0 0 1>}
  n54 = NamedRerouteUsage()
  n55 = NamedRerouteUsage()
  n56 = NamedRerouteUsage()
  n57 = NamedRerouteUsage()
  n58 = Constant()
  n59 = NamedRerouteUsage()
  n60 = NamedRerouteUsage()
  NamedRerouteDeclaration(None=Saturate(None=MaterialFunctionCall!(UVs=@n54, Size=@n55, Outline Thickness=@n56, Glow Min=@n57, Glow Max=@n58, Corner Roundness=@n59, Keep Aspect Ratio=@n60) {materialFunction=/FortUILibrary/UGC/MaterialFunctions/MF_UGCUI_Box}.Outline)) {name=Shape Stroke, nodeColor=<rgba 0 0 0 1>}
  n61 = NamedRerouteUsage()
  n62 = NamedRerouteUsage()
  n63 = NamedRerouteUsage()
  NamedRerouteDeclaration(None=LinearInterpolate(A=@n61, B=LinearInterpolate(A=@n61, B=@n62, Alpha=@n63), Alpha=NamedRerouteUsage())) {name=Idle Color or Gradient, nodeColor=<rgba 0 0 0 1>}
  n64 = VectorParameter() {
   desc="Default = Black\r\n\r\nThe main color applied to transparent pixels. \r\n\r\nWhen you reduce the transparency of an element, it fades on top of this color.\r\nThis means that this color will influence the color of your elements when they become transparent. \r\n"
   group="[3] Colors"
   parameterName=Color for Transparency 1
   sortPriority=5
   defaultValue=<rgba 0 0 0 1>
  }
  NamedRerouteDeclaration(None=Saturate(None=MaterialFunctionCall(X=@n64.R, Y=@n64.R, Z=@n64.R, A=@n64.R) {materialFunction=&MakeFloat4}.Result)) {name=Color for Transparency 1, nodeColor=<rgba 0 0 0 1>}
  n65 = VectorParameter() {
   desc="Default = Black\r\n\r\nSecondary color for transparency when the gradient is Active.\r\n"
   group="[3] Colors"
   parameterName="Color for Transparency 2 (gradient on)"
   sortPriority=6
   defaultValue=<rgba 0 0 0 1>
  }
  NamedRerouteDeclaration(None=Saturate(None=MaterialFunctionCall(X=@n65.R, Y=@n65.R, Z=@n65.R, A=@n65.R) {materialFunction=&MakeFloat4}.Result)) {name="Color for Transparency 2 (Gradient on)", nodeColor=<rgba 0 0 0 1>}
  n66 = NamedRerouteUsage()
  n67 = NamedRerouteUsage()
  n68 = NamedRerouteUsage()
  NamedRerouteDeclaration(None=LinearInterpolate(A=@n66, B=LinearInterpolate(A=@n66, B=@n67, Alpha=@n68), Alpha=NamedRerouteUsage())) {name=Color for Transparency - Color or Gradient, nodeColor=<rgba 0 0 0 1>}
  n69 = NamedRerouteUsage()
  n70 = Distance(A=ConstantBiasScale(None=@n69) {bias=-0.5, scale=2}, B=Constant())
  n71 = NamedRerouteUsage()
  n72 = NamedRerouteUsage()
  n73 = Subtract(A=LinearInterpolate(A=ComponentMask(None=@n69) {r=true}, B=Subtract(A=@n70, B=@n71), Alpha=NamedRerouteUsage()), B=@n72)
  n74 = Subtract(A=NamedRerouteUsage(), B=@n72)
  n75 = If(A=@n74, "A > B"=@n74, "A == B"=Constant() {r=0}, "A < B"=@n74)
  NamedRerouteDeclaration(None=Saturate(None=Divide(A=@n73, B=@n75))) {name=gradient, nodeColor=<rgba 0 0 0 1>}
  n76 = NamedRerouteUsage()
  n77 = NamedRerouteUsage()
  n78 = NamedRerouteUsage()
  NamedRerouteDeclaration(None=LinearInterpolate(A=LinearInterpolate(A=@n76, B=@n77, Alpha=@n78), B=NamedRerouteUsage(), Alpha=NamedRerouteUsage())) {name="Hovered  Final Color", nodeColor=<rgba 0 0 0 1>}
  n79 = NamedRerouteUsage()
  n80 = NamedRerouteUsage()
  n81 = NamedRerouteUsage()
  NamedRerouteDeclaration(None=LinearInterpolate(A=LinearInterpolate(A=@n79, B=@n80, Alpha=@n81), B=NamedRerouteUsage(), Alpha=NamedRerouteUsage())) {name="Focused  Final Color", nodeColor=<rgba 0 0 0 1>}
  n82 = NamedRerouteUsage()
  n83 = NamedRerouteUsage()
  n84 = NamedRerouteUsage()
  NamedRerouteDeclaration(None=LinearInterpolate(A=LinearInterpolate(A=@n82, B=@n83, Alpha=@n84), B=NamedRerouteUsage(), Alpha=NamedRerouteUsage())) {name="Pressed  Final Color", nodeColor=<rgba 0 0 0 1>}
  n85 = NamedRerouteUsage()
  n86 = NamedRerouteUsage()
  n87 = NamedRerouteUsage()
  NamedRerouteDeclaration(None=LinearInterpolate(A=LinearInterpolate(A=@n85, B=@n86, Alpha=@n87), B=NamedRerouteUsage(), Alpha=NamedRerouteUsage())) {name="Disabled  Final Color", nodeColor=<rgba 0 0 0 1>}
  n88 = MaterialFunctionCall() {materialFunction=&GetUserInterfaceUV}
  NamedRerouteDeclaration(None=LinearInterpolate(A=@n88.9-Slice UV, B=@n88.Normalized UV, Alpha=NamedRerouteUsage())) {name=Common UV, nodeColor=<rgba 1 0.0036 0 1>}
  n89 = ScalarParameter() {
   defaultValue=1
   desc="Default Value = 1.00 \r\nRange: 0.0-1.0\r\n\r\nSet the transparency between 0 (transparent) and 1 (full opaque).\r\n"
   group="[1] Fill"
   parameterName=Fill Transparency
   sliderMax=1
   sortPriority=2
  }
  NamedRerouteDeclaration(None=@n89) {name=Idle Fill Transparency, nodeColor=<rgba 0 0 0 1>}
  n90 = ScalarParameter() {
   desc="Default value =  0.0 \r\nRange: -1 to 1\r\n\r\nRotate your Shape.\r\n\r\nWARNING:\r\nRotate may deform the shape unless you:\r\n\r\n- Use a different UV Scaling Type to Scale to Fit or Scale to Fill.\r\n\r\n- Set squared widget size\r\n"
   group="[1] Fill"
   parameterName=Rotation Angle
   sliderMax=1
   sliderMin=-1
   sortPriority=17
  }
  NamedRerouteDeclaration(None=@n90) {name=Shape Rotation Angle, nodeColor=<rgba 0 0 0 1>}
  n91 = ScalarParameter() {
   desc="0.0 = No (Default)\r\nGreater than 0.0 = Yes\r\n\r\nWhen Disabled, the Stroke is computed on top of the Fill. \r\n\r\nWhen Enabled, the Stroke will be subtracted from the Fill. Which means that behind the stroke, there will be no fill anymore. If you reduce the Stroke opacity, you will see that the fill is empty / hollowed out where the stroke is.\r\n"
   group="[2] Stroke"
   parameterName=Subtract Stroke From Fill
   sliderMax=1
   sortPriority=7
  }
  NamedRerouteDeclaration(None=Saturate(None=Ceil(None=@n91))) {name=Subtract Stroke From Fill, nodeColor=<rgba 0 0 0 1>}
  n92 = VectorParameter() {
   desc="Set the dedicated stroke main color. \r\n\r\nYou can use the alpha chanel of the color to reduce the opacity of the color only. (can be useful with gradient).\r\n"
   group="[3] Colors"
   parameterName=Stroke Color 1
   sortPriority=3
   defaultValue=<rgba 0 0.333 1 1>
  }
  NamedRerouteDeclaration(None=Saturate(None=MaterialFunctionCall(X=@n92.R, Y=@n92.R, Z=@n92.R, A=@n92.R) {materialFunction=&MakeFloat4}.Result)) {name=Idle Stroke Color 1, nodeColor=<rgba 0 0 0 1>}
  n93 = VectorParameter() {
   desc="Set the dedicated stroke secondary color. (gradient on) \r\n\r\nYou can use the alpha chanel of the color to reduce the opacity of the color only. (can be useful with gradient).\r\n"
   group="[3] Colors"
   parameterName="Stroke Color 2 (gradient on)"
   sortPriority=4
   defaultValue=<rgba 1 0.333 0 1>
  }
  NamedRerouteDeclaration(None=Saturate(None=MaterialFunctionCall(X=@n93.R, Y=@n93.R, Z=@n93.R, A=@n93.R) {materialFunction=&MakeFloat4}.Result)) {name="Idle Stroke Color 2 (gradient on)", nodeColor=<rgba 0 0 0 1>}
  n94 = VectorParameter() {
   desc="Set the dedicated stroke main color. \r\n\r\nYou can use the alpha chanel of the color to reduce the opacity of the color only. (can be useful with gradient).\r\n"
   group="[4a] State-Hovered"
   parameterName=Hovered Stroke Color 1
   sortPriority=3
   defaultValue=<rgba 1 0.932 0.6562 1>
  }
  NamedRerouteDeclaration(None=Saturate(None=MaterialFunctionCall(X=@n94.R, Y=@n94.R, Z=@n94.R, A=@n94.R) {materialFunction=&MakeFloat4}.Result)) {name=Hovered Stroke Color 1, nodeColor=<rgba 0 0 0 1>}
  n95 = VectorParameter() {
   desc="Set the dedicated stroke secondary color. (gradient on) \r\n\r\nYou can use the alpha chanel of the color to reduce the opacity of the color only. (can be useful with gradient).\r\n"
   group="[4a] State-Hovered"
   parameterName="Hovered Stroke Color 2 (gradient on)"
   sortPriority=4
   defaultValue=<rgba 1 0.7014 0.7717 1>
  }
  NamedRerouteDeclaration(None=Saturate(None=MaterialFunctionCall(X=@n95.R, Y=@n95.R, Z=@n95.R, A=@n95.R) {materialFunction=&MakeFloat4}.Result)) {name="Hovered Stroke Color 2 (gradient on)", nodeColor=<rgba 0 0 0 1>}
  n96 = VectorParameter() {
   desc="Set the dedicated stroke main color. \r\n\r\nYou can use the alpha chanel of the color to reduce the opacity of the color only. (can be useful with gradient).\r\n"
   group="[4b] State-Focused"
   parameterName=Focused Stroke Color 1
   sortPriority=3
   defaultValue=<rgba 1 0.932 0.6562 1>
  }
  NamedRerouteDeclaration(None=Saturate(None=MaterialFunctionCall(X=@n96.R, Y=@n96.R, Z=@n96.R, A=@n96.R) {materialFunction=&MakeFloat4}.Result)) {name=Focused Stroke Color 1, nodeColor=<rgba 0 0 0 1>}
  n97 = VectorParameter() {
   desc="Set the dedicated stroke secondary color. (gradient on) \r\n\r\nYou can use the alpha chanel of the color to reduce the opacity of the color only. (can be useful with gradient).\r\n"
   group="[4b] State-Focused"
   parameterName="Focused Stroke Color 2 (gradient on)"
   sortPriority=4
   defaultValue=<rgba 1 0.7014 0.7717 1>
  }
  NamedRerouteDeclaration(None=Saturate(None=MaterialFunctionCall(X=@n97.R, Y=@n97.R, Z=@n97.R, A=@n97.R) {materialFunction=&MakeFloat4}.Result)) {name="Focused Stroke Color 2 (gradient on)", nodeColor=<rgba 0 0 0 1>}
  n98 = VectorParameter() {
   desc="Set the dedicated stroke main color. \r\n\r\nYou can use the alpha chanel of the color to reduce the opacity of the color only. (can be useful with gradient).\r\n"
   group="[4c] State-Pressed"
   parameterName=Pressed Stroke Color 1
   sortPriority=3
   defaultValue=<rgba 0.7292 0.7126 0.6456 1>
  }
  NamedRerouteDeclaration(None=Saturate(None=MaterialFunctionCall(X=@n98.R, Y=@n98.R, Z=@n98.R, A=@n98.R) {materialFunction=&MakeFloat4}.Result)) {name=Pressed Stroke Color 1, nodeColor=<rgba 0 0 0 1>}
  n99 = VectorParameter() {
   desc="Set the dedicated stroke secondary color. (gradient on) \r\n\r\nYou can use the alpha chanel of the color to reduce the opacity of the color only. (can be useful with gradient).\r\n"
   group="[4c] State-Pressed"
   parameterName="Pressed Stroke Color 2 (gradient on)"
   sortPriority=4
   defaultValue=<rgba 0.8924 0.787 0.8118 1>
  }
  NamedRerouteDeclaration(None=Saturate(None=MaterialFunctionCall(X=@n99.R, Y=@n99.R, Z=@n99.R, A=@n99.R) {materialFunction=&MakeFloat4}.Result)) {name="Pressed Stroke Color 2 (gradient on)", nodeColor=<rgba 0 0 0 1>}
  n100 = VectorParameter() {
   desc="Set the dedicated stroke main color. \r\n\r\nYou can use the alpha chanel of the color to reduce the opacity of the color only. (can be useful with gradient).\r\n"
   group="[4d] State-Disabled"
   parameterName=Disabled Stroke Color 1
   sortPriority=3
   defaultValue=<rgba 0.1111 0.1111 0.1111 1>
  }
  NamedRerouteDeclaration(None=Saturate(None=MaterialFunctionCall(X=@n100.R, Y=@n100.R, Z=@n100.R, A=@n100.R) {materialFunction=&MakeFloat4}.Result)) {name=Disabled Stroke Color 1, nodeColor=<rgba 0 0 0 1>}
  n101 = VectorParameter() {
   desc="Set the dedicated stroke secondary color. (gradient on) \r\n\r\nYou can use the alpha chanel of the color to reduce the opacity of the color only. (can be useful with gradient).\r\n"
   group="[4d] State-Disabled"
   parameterName="Disabled Stroke Color 2 (gradient on)"
   sortPriority=4
   defaultValue=<rgba 0.0451 0.0451 0.0451 1>
  }
  NamedRerouteDeclaration(None=Saturate(None=MaterialFunctionCall(X=@n101.R, Y=@n101.R, Z=@n101.R, A=@n101.R) {materialFunction=&MakeFloat4}.Result)) {name="Disabled Stroke Color 2 (gradient on)", nodeColor=<rgba 0 0 0 1>}
  n102 = NamedRerouteUsage()
  n103 = NamedRerouteUsage()
  n104 = NamedRerouteUsage()
  NamedRerouteDeclaration(None=LinearInterpolate(A=@n102, B=LinearInterpolate(A=@n102, B=@n103, Alpha=@n104), Alpha=NamedRerouteUsage())) {name=Hovered Stroke Color or Gradient, nodeColor=<rgba 0 0 0 1>}
  n105 = NamedRerouteUsage()
  n106 = NamedRerouteUsage()
  n107 = NamedRerouteUsage()
  NamedRerouteDeclaration(None=LinearInterpolate(A=@n105, B=LinearInterpolate(A=@n105, B=@n106, Alpha=@n107), Alpha=NamedRerouteUsage())) {name=Focused Stroke Color or Gradient, nodeColor=<rgba 0 0 0 1>}
  n108 = NamedRerouteUsage()
  n109 = NamedRerouteUsage()
  n110 = NamedRerouteUsage()
  NamedRerouteDeclaration(None=LinearInterpolate(A=@n108, B=LinearInterpolate(A=@n108, B=@n109, Alpha=@n110), Alpha=NamedRerouteUsage())) {name=Pressed Stroke Color or Gradient, nodeColor=<rgba 0 0 0 1>}
  n111 = NamedRerouteUsage()
  n112 = NamedRerouteUsage()
  n113 = NamedRerouteUsage()
  NamedRerouteDeclaration(None=LinearInterpolate(A=@n111, B=LinearInterpolate(A=@n111, B=@n112, Alpha=@n113), Alpha=NamedRerouteUsage())) {name=Disabled Stroke Color or Gradient, nodeColor=<rgba 0 0 0 1>}
  n114 = NamedRerouteUsage()
  n115 = NamedRerouteUsage()
  n116 = NamedRerouteUsage()
  NamedRerouteDeclaration(None=LinearInterpolate(A=@n114, B=LinearInterpolate(A=@n114, B=@n115, Alpha=@n116), Alpha=NamedRerouteUsage())) {name=Idle Stroke Color or Gradient, nodeColor=<rgba 0 0 0 1>}
  n117 = MaterialFunctionCall(
   UVs=NamedRerouteUsage()
   Size=NamedRerouteUsage()
   Outline Thickness=NamedRerouteUsage()
   Glow Min=NamedRerouteUsage()
   Glow Max=Constant()
   Corner Roundness=NamedRerouteUsage()
   Keep Aspect Ratio=NamedRerouteUsage()
  ) {materialFunction=&MF_UGCUI_Box}
  n118 = NamedRerouteUsage()
  NamedRerouteDeclaration(None=Saturate(None=Subtract(A=@n117.Outline, B=@n118))) {name=Gap, nodeColor=<rgba 0 0 0 1>}
  n119 = MaterialFunctionCall() {materialFunction=&GetUserInterfaceUV}
  n120 = ComponentMask(None=@n119.Pixel Size) {r=true}
  n121 = ComponentMask(None=@n119.Pixel Size) {g=true}
  NamedRerouteDeclaration(None=Min(
    A=NamedRerouteDeclaration(None=@n120) {name=pixelSize.x, nodeColor=<rgba 0 0.8125 1 1>}
    B=NamedRerouteDeclaration(None=@n121) {name=pixelSize.y, nodeColor=<rgba 1 0.5391 0 1>}
   )) {name=Pixel Size Min, nodeColor=<rgba 0.1484 0 1 1>}
  n122 = MaterialFunctionCall() {materialFunction=&ScreenResolution}
  n123 = ComponentMask(None=@n122.Visible Resolution) {r=true}
  n124 = ComponentMask(None=@n122.Visible Resolution) {g=true}
  NamedRerouteDeclaration(None=Min(
    A=NamedRerouteDeclaration(None=@n123) {name=ScreenResolution.x, nodeColor=<rgba 0 0.8125 1 1>}
    B=NamedRerouteDeclaration(None=@n124) {name=ScreenResolution.y, nodeColor=<rgba 1 0.5391 0 1>}
   )) {name=Screen Resolution Min, nodeColor=<rgba 0.0309 0.0675 1 1>}
  n125 = ScalarParameter() {
   defaultValue=0.98
   desc="Default Value = 0.98 \r\nRange: 0-1\r\n\r\nChange how sharp the edges of the shape are.\r\n"
   group="[1] Fill"
   parameterName=Sharpness
   sliderMax=1
   sortPriority=7
  }
  MaterialExpressionRemap_0 = MaterialXRemap!(None=@n125, InputLow, InputHigh, TargetLow, TargetHigh) {targetHighDefault=0, targetLowDefault=-0.1}
  NamedRerouteDeclaration(None=Clamp(None=@MaterialExpressionRemap_0) {desc=0 can not work..., maxDefault=-0.0001, minDefault=-0.1}) {name=Shape Sharpness, nodeColor=<rgba 0 0 0 1>}
  NamedRerouteDeclaration(None=MaterialFunctionCall(UVs=NamedRerouteUsage(), "Rotation Angle (0-1)"=NamedRerouteUsage()) {materialFunction=&CustomRotator}.Rotated Values) {name=Shape Rotated UV, nodeColor=<rgba 1 0.0036 0 1>}
  n126 = ScalarParameter() {
   defaultValue=1
   desc="Default Value = 1.0\r\n\r\nInfo:\r\nThe Stroke Scale is relative to the Scale parameter. \r\n"
   group="[2] Stroke"
   parameterName=Stroke Scale
   sortPriority=3
  }
  NamedRerouteDeclaration(None=@n126) {name=Stroke Scale, nodeColor=<rgba 0 0 0 1>}
  n127 = ScalarParameter() {
   desc="\"Reference Size\" but named \"Consistant Stroke Thikness\" for players 'cause it is replacing the stroke thickness.\r\nif 0, not activated. \r\nfor any other value,  stroke thickness = 1 is 1 for this reference size."
   group="[2] Stroke"
   parameterName=Stroke Consistant Thikness on/off
   sliderMax=1
   sortPriority=2
  }
  NamedRerouteDeclaration(None=@n127) {name=Keep Consistant Stroke Thickness, nodeColor=<rgba 0 0 0 1>}
  n128 = Ceil(None=NamedRerouteUsage())
  NamedRerouteDeclaration(None=LinearInterpolate(A=NamedRerouteUsage(), B=NamedRerouteUsage(), Alpha=Saturate(None=@n128))) {name=Final stroke Thickness, nodeColor=<rgba 0 0 0 1>}
  n129 = Multiply(A=NamedRerouteUsage()) {constB=0.1}
  n130 = NamedRerouteUsage()
  NamedRerouteDeclaration(None=Divide(A=Multiply(A=@n129, B=@n130), B=NamedRerouteUsage())) {name=Consistant Stroke Thickness, nodeColor=<rgba 0 0 0 1>}
  n131 = VectorParameter() {
   desc="Set the main color. \r\n\r\nYou can use the alpha chanel of the color to reduce the opacity of the color only. (can be useful with gradient).\r\n"
   group="[4a] State-Hovered"
   parameterName=Hovered Color 1
   sortPriority=1
   defaultValue=<rgba 0.4931 0.6893 1 1>
  }
  NamedRerouteDeclaration(None=Saturate(None=MaterialFunctionCall(X=@n131.R, Y=@n131.R, Z=@n131.R, A=@n131.R) {materialFunction=&MakeFloat4}.Result)) {name=Hovered Color 1, nodeColor=<rgba 0 0 0 1>}
  n132 = NamedRerouteUsage()
  n133 = NamedRerouteUsage()
  n134 = ScalarParameter() {
   defaultValue=1
   desc="Default value = 1.0\r\n\r\nThe stroke corners roundness is relative to the 4 Corners Roundness Parameters. \r\n\r\nBecause of that, if you change the scale or the size of the stroke using the width and height parameters, you will observe that the stroke corners can have a different roundness than the fill, and then become unsightly.\r\n\r\nThis option let you fine tune the Stroke Corners Roundness depending the Stroke Scale and Size you'll may set.\r\n"
   group="[2] Stroke"
   parameterName=Stroke Corner Roundness Scale
   sortPriority=6
  }
  NamedRerouteDeclaration(None=Multiply(A=Multiply(A=@n132, B=@n133) {desc="Make the stroke Corner roundness relative \r\nto the Corners Roundness Parameter."}, B=@n134)) {name=Stroke Corner Roundness, nodeColor=<rgba 0 0 0 1>}
  n135 = VectorParameter() {
   desc="Set the econdary color (gradient on). \r\n\r\nYou can use the alpha chanel of the color to reduce the opacity of the color only. (can be useful with gradient).\r\n"
   group="[4a] State-Hovered"
   parameterName="Hovered Color 2 (gradient on)"
   sortPriority=2
   defaultValue=<rgba 0.3395 0.2953 0.4549 1>
  }
  NamedRerouteDeclaration(None=Saturate(None=MaterialFunctionCall(X=@n135.R, Y=@n135.R, Z=@n135.R, A=@n135.R) {materialFunction=&MakeFloat4}.Result)) {name="Hovered Color 2 (gradient on)", nodeColor=<rgba 0 0 0 1>}
  n136 = ScalarParameter() {
   defaultValue=1
   desc="Default Value = 1.0\r\n\r\nInfo:\r\nThe Stroke Width is relative to the Width parameter. \r\n"
   group="[2] Stroke"
   parameterName=Stroke Width
   sortPriority=4
  }
  n137 = NamedRerouteUsage()
  n138 = Multiply(A=@n136, B=@n137)
  n139 = ScalarParameter() {
   defaultValue=1
   desc="Default Value = 1.0\r\n\r\nInfo:\r\nThe Stroke Height is relative to the Height parameter. \r\n"
   group="[2] Stroke"
   parameterName=Stroke Height
   sortPriority=5
  }
  n140 = Multiply(A=@n137, B=@n139)
  NamedRerouteDeclaration(None=Multiply(A=AppendVector(A=@n138, B=@n140), B=NamedRerouteUsage())) {name=Stroke Size, nodeColor=<rgba 0 0 0 1>}
  n141 = VectorParameter() {
   desc="Set the main color. \r\n\r\nYou can use the alpha chanel of the color to reduce the opacity of the color only. (can be useful with gradient).\r\n"
   group="[4b] State-Focused"
   parameterName=Focused Color 1
   sortPriority=1
   defaultValue=<rgba 0.4931 0.6893 1 1>
  }
  NamedRerouteDeclaration(None=Saturate(None=MaterialFunctionCall(X=@n141.R, Y=@n141.R, Z=@n141.R, A=@n141.R) {materialFunction=&MakeFloat4}.Result)) {name=Focused Color 1, nodeColor=<rgba 0 0 0 1>}
  n142 = ScalarParameter() {
   defaultValue=1
   desc="Default Value = 1.0\r\n\r\nScale your shape from 0 to 1 (100%) \r\n"
   group="[1] Fill"
   parameterName=Scale
   sliderMax=1
   sortPriority=4
  }
  NamedRerouteDeclaration(None=@n142) {name=Scale, nodeColor=<rgba 0 0 0 1>}
  n143 = VectorParameter() {
   desc="Set the econdary color (gradient on). \r\n\r\nYou can use the alpha chanel of the color to reduce the opacity of the color only. (can be useful with gradient).\r\n"
   group="[4b] State-Focused"
   parameterName="Focused Color 2 (gradient on)"
   sortPriority=2
   defaultValue=<rgba 0.3395 0.2953 0.4549 1>
  }
  NamedRerouteDeclaration(None=Saturate(None=MaterialFunctionCall(X=@n143.R, Y=@n143.R, Z=@n143.R, A=@n143.R) {materialFunction=&MakeFloat4}.Result)) {name="Focused Color 2 (gradient on)", nodeColor=<rgba 0 0 0 1>}
  n144 = ScalarParameter() {
   defaultValue=1
   desc="Lower than 1.0 =  Let the box stretch\r\n1.0 = Keep the corners aspect ratio (default)\r\n\r\nWARNING:\r\nIf you switch to another UV Scaling than the Stretch one, then you have to let this 'Keep Aspect Ratio' value to 0 to avoid cancelling the expected result. \r\n"
   group="[1] Fill"
   parameterName=Keep Corners Aspect Ratio
   sliderMax=1
   sortPriority=12
  }
  NamedRerouteDeclaration(None=@n144) {name=Corner Aspect Ratio, nodeColor=<rgba 0 0 0 1>}
  n145 = VectorParameter() {
   desc="Set the main color. \r\n\r\nYou can use the alpha chanel of the color to reduce the opacity of the color only. (can be useful with gradient).\r\n"
   group="[4c] State-Pressed"
   parameterName=Pressed Color 1
   sortPriority=1
   defaultValue=<rgba 1 1 1 1>
  }
  NamedRerouteDeclaration(None=Saturate(None=MaterialFunctionCall(X=@n145.R, Y=@n145.R, Z=@n145.R, A=@n145.R) {materialFunction=&MakeFloat4}.Result)) {name=Pressed Color 1, nodeColor=<rgba 0 0 0 1>}
  n146 = VectorParameter() {
   desc="Set the econdary color (gradient on). \r\n\r\nYou can use the alpha chanel of the color to reduce the opacity of the color only. (can be useful with gradient).\r\n"
   group="[4c] State-Pressed"
   parameterName="Pressed Color 2 (gradient on)"
   sortPriority=2
   defaultValue=<rgba 0.5712 0.5712 0.5712 1>
  }
  NamedRerouteDeclaration(None=Saturate(None=MaterialFunctionCall(X=@n146.R, Y=@n146.R, Z=@n146.R, A=@n146.R) {materialFunction=&MakeFloat4}.Result)) {name="Pressed Color 2 (gradient on)", nodeColor=<rgba 0 0 0 1>}
  n147 = NamedRerouteUsage()
  n148 = NamedRerouteUsage()
  NamedRerouteDeclaration(None=MaterialFunctionCall!(
    "Value to Ease (0-1)"=MaterialFunctionCall!(No Time Static Value, Speed=@n147, Time Style Switch Index=@n148) {materialFunction=/FortUILibrary/UGC/MaterialFunctions/MF_UGCUI_TimeSwitcher}.Selected Time Style
    Easing Curve Switch Index=NamedRerouteUsage()
   ) {materialFunction=/FortUILibrary/UGC/MaterialFunctions/MF_UGCUI_EaseCurveSwitcher}.Selected Easing Curve) {name=Gradient Animated Rotation, nodeColor=<rgba 0 0 0 1>}
  n149 = ScalarParameter() {
   desc="0 = Gradient off (default)\r\nGreater than 0 = Gradient on\r\n\r\nWhen enabled, Color 1 and Color 2 are used to make the gradient. \r\nWhen disabled, Color 1 is used only. \r\n"
   group="[3] Colors"
   parameterName=Gradient on / off
   sliderMax=1
   sortPriority=7
  }
  NamedRerouteDeclaration(None=Saturate(None=Ceil(None=@n149))) {name=Gradient On / Off, nodeColor=<rgba 0 0 0 1>}
  n150 = VectorParameter() {
   desc="Set the main color. \r\n\r\nYou can use the alpha chanel of the color to reduce the opacity of the color only. (can be useful with gradient).\r\n"
   group="[4d] State-Disabled"
   parameterName=Disabled Color 1
   sortPriority=1
   defaultValue=<rgba 0.1997 0.1997 0.1997 1>
  }
  NamedRerouteDeclaration(None=Saturate(None=MaterialFunctionCall(X=@n150.R, Y=@n150.R, Z=@n150.R, A=@n150.R) {materialFunction=&MakeFloat4}.Result)) {name=Disabled Color 1, nodeColor=<rgba 0 0 0 1>}
  n151 = VectorParameter() {
   desc="Set the main color. \r\n\r\nYou can use the alpha chanel of the color to reduce the opacity of the color only. (can be useful with gradient).\r\n"
   group="[3] Colors"
   parameterName=Color 1
   sortPriority=1
   defaultValue=<rgba 1 1 1 1>
  }
  n152 = MaterialFunctionCall(X=@n151.R, Y=@n151.R, Z=@n151.R, A=@n151.R) {
   desc="Needed to get the Alpha of the color and use it \r\nin the MF_UGCUI_ComputeAlpha Material Function"
   materialFunction=&MakeFloat4
  }
  NamedRerouteDeclaration(None=Saturate(None=@n152.Result)) {name=Idle Color 1, nodeColor=<rgba 0 0 0 1>}
  n153 = VectorParameter() {
   desc="Set the secondary color (gradient on). \r\n\r\nYou can use the alpha chanel of the color to reduce the opacity of the color only. (can be useful with gradient).\r\n"
   group="[3] Colors"
   parameterName="Color 2 (gradient on)"
   sortPriority=2
   defaultValue=<rgba 0 0.33 1 1>
  }
  NamedRerouteDeclaration(None=Saturate(None=MaterialFunctionCall(X=@n153.R, Y=@n153.R, Z=@n153.R, A=@n153.R) {materialFunction=&MakeFloat4}.Result)) {name="Idle Color 2 (gradient on)", nodeColor=<rgba 0 0 0 1>}
  n154 = ScalarParameter() {
   desc="0.0 = Use linear gradient (default)\r\nGreater than 0.0 = Use radial gradient\r\n"
   group="[3] Colors"
   parameterName=Select Linear or Radial Gradient
   sliderMax=1
   sortPriority=8
  }
  NamedRerouteDeclaration(None=Saturate(None=Ceil(None=@n154))) {name=Linear or Radial Gradient, nodeColor=<rgba 0 0 0 1>}
  n155 = VectorParameter() {
   desc="Set the econdary color (gradient on). \r\n\r\nYou can use the alpha chanel of the color to reduce the opacity of the color only. (can be useful with gradient).\r\n"
   group="[4d] State-Disabled"
   parameterName="Disabled Color 2 (gradient on)"
   sortPriority=2
   defaultValue=<rgba 0.0677 0.0677 0.0677 1>
  }
  NamedRerouteDeclaration(None=Saturate(None=MaterialFunctionCall(X=@n155.R, Y=@n155.R, Z=@n155.R, A=@n155.R) {materialFunction=&MakeFloat4}.Result)) {name="Disabled Color 2 (gradient on)", nodeColor=<rgba 0 0 0 1>}
  n156 = ScalarParameter() {desc="Default Value = 0.0\r\n\r\nWhere the gradient must start\r\n", group="[3] Colors", parameterName=Gradient Start, sortPriority=10}
  NamedRerouteDeclaration(None=@n156) {name=Gradient Start, nodeColor=<rgba 0 0 0 1>}
  n157 = ScalarParameter() {
   defaultValue=0.015
   desc="Default Value = 0.015\r\n\r\nChange  the thickness of the stroke.\r\n"
   group="[2] Stroke"
   parameterName=Stroke Thickness
   sliderMax=1
   sortPriority=2
  }
  n158 = ScalarParameter() {
   desc="Default value = 0\r\n\r\nAdd a gap between stroke and fill. The gap position depends on the stroke thickness, size and scale.\r\n\r\nTIPS:\r\nthe gap exists even if the stroke transparency is equal to 0. In fact, applying a gap means digging into the filling.\r\n\r\nSet 1 to the parameter 'Subtract Stroke From Fill' will hide the the fill where the stroke is supposed to be displayed (even if the stroke transparency is equal to 0).\r\n"
   group="[2] Stroke"
   parameterName=Gap Thickness
   sliderMax=1
   sortPriority=3
  }
  NamedRerouteDeclaration(None=Add(A=@n157, B=@n158)) {name=GapThickness, nodeColor=<rgba 0 0 0 1>}
  n159 = ScalarParameter() {
   defaultValue=1
   desc="Default Value = 1.0\r\n\r\nWhere the gradient must end\r\n"
   group="[3] Colors"
   parameterName=Gradient End
   sortPriority=11
  }
  NamedRerouteDeclaration(None=@n159) {name=Gradient End, nodeColor=<rgba 0 0 0 1>}
  n160 = NamedRerouteUsage()
  n161 = NamedRerouteUsage()
  n162 = NamedRerouteUsage()
  NamedRerouteDeclaration(None=LinearInterpolate(A=LinearInterpolate(A=@n160, B=@n161, Alpha=@n162), B=NamedRerouteUsage(), Alpha=NamedRerouteUsage())) {name=Idle Final Color, nodeColor=<rgba 0 0 0 1>}
  n163 = NamedRerouteUsage()
  n164 = NamedRerouteUsage()
  n165 = NamedRerouteUsage()
  NamedRerouteDeclaration(None=LinearInterpolate(A=@n163, B=LinearInterpolate(A=@n163, B=@n164, Alpha=@n165), Alpha=NamedRerouteUsage())) {name=Hovered Color or Gradient, nodeColor=<rgba 0 0 0 1>}
  n166 = NamedRerouteUsage()
  n167 = NamedRerouteUsage()
  n168 = NamedRerouteUsage()
  NamedRerouteDeclaration(None=LinearInterpolate(A=@n166, B=LinearInterpolate(A=@n166, B=@n167, Alpha=@n168), Alpha=NamedRerouteUsage())) {name=Focused Color or Gradient, nodeColor=<rgba 0 0 0 1>}
  n169 = ScalarParameter() {
   defaultValue=1
   desc="Default = 1.0\r\n\r\nApplies to the Radial Gradient only. Scale up or down the gradient circle.\r\n"
   group="[3] Colors"
   parameterName="Gradient Size Factor (Radial Only)"
   sortPriority=12
  }
  NamedRerouteDeclaration(None=Subtract(A=@n169) {desc=Make more sense for the user}) {name=Radial Gradient Scale, nodeColor=<rgba 0 0 0 1>}
  n170 = NamedRerouteUsage()
  n171 = NamedRerouteUsage()
  n172 = NamedRerouteUsage()
  NamedRerouteDeclaration(None=LinearInterpolate(A=@n170, B=LinearInterpolate(A=@n170, B=@n171, Alpha=@n172), Alpha=NamedRerouteUsage())) {name=Pressed Color or Gradient, nodeColor=<rgba 0 0 0 1>}
  NamedRerouteDeclaration(None=NamedRerouteUsage()) {name=Shape UV, nodeColor=<rgba 1 0.0036 0 1>}
  n173 = ScalarParameter() {
   desc="Default Value = 0.0\r\n\r\nOffset the gradient from its center to a direction or another. 0 (default) is the center.\r\n\r\nTips:\r\nWhen Radial Gradient is selected, you need to set an offset to see the gradient rotation, otherwise, the gradient will look static (rotating on its very center). \r\n"
   group="[3] Colors"
   parameterName=Gradient Offset
   sortPriority=13
  }
  NamedRerouteDeclaration(None=@n173) {name=Gradient Offset, nodeColor=<rgba 0 0 0 1>}
  n174 = NamedRerouteUsage()
  n175 = NamedRerouteUsage() {desc="If 0, the UV scaling is comming \r\nfrom the shape. This one is ignored."}
  n176 = Saturate(None=Floor(None=@n175))
  NamedRerouteDeclaration(None=MaterialFunctionCall!(UV=NamedRerouteUsage(), Texture Size, UV Scaling Switch Index=LinearInterpolate(A=@n174, B=@n175, Alpha=@n176)) {materialFunction=/FortUILibrary/UGC/MaterialFunctions/MF_UGCUI_UVScalingSwitcher}.Selected Scaled UV) {name=Gradient Scaled UV, nodeColor=<rgba 1 0.0036 0 1>}
  n177 = ScalarParameter() {
   desc="0.0 = Use the UV Scaling from the Shape  (default) \r\n1.0 = Stretch\r\n2.0 = Scaled to Fit \r\n3.0 = Scaled to Fill \r\n\r\nTips:\r\nthe 2 and 3 methods keep the aspect ratio of the Gradient. \r\n\r\nWARNING:\r\n\r\nRotation on streched widget NEEDS a scale to fit or to fill to avoid distortions.\r\n"
   group="[3] Colors"
   parameterName=Select the Gradient Scaling Method
   sliderMax=3
   sortPriority=9
  }
  NamedRerouteDeclaration(None=@n177) {name=Gradient UV Scaling, nodeColor=<rgba 0 0 0 1>}
  n178 = NamedRerouteUsage()
  n179 = NamedRerouteUsage()
  n180 = NamedRerouteUsage()
  NamedRerouteDeclaration(None=LinearInterpolate(A=@n178, B=LinearInterpolate(A=@n178, B=@n179, Alpha=@n180), Alpha=NamedRerouteUsage())) {name=Disabled Color or Gradient, nodeColor=<rgba 0 0 0 1>}
  NamedRerouteDeclaration(None=NamedRerouteUsage()) {name=gradient UV, nodeColor=<rgba 1 0.0036 0 1>}
  n181 = NamedRerouteUsage()
  n182 = Add(A=NamedRerouteUsage(), B=NamedRerouteUsage())
  NamedRerouteDeclaration(None=Subtract(
    A=MaterialFunctionCall(UVs=@n181, "Rotation Angle (0-1)"=@n182) {materialFunction=&CustomRotator}.Rotated Values
    B=NamedRerouteUsage()
   )) {name=gradient Rotated UV, nodeColor=<rgba 1 0.0036 0 1>}
  n183 = ScalarParameter() {
   desc="Default value =  0.0\r\nRange: -1.0 to 1.0\r\n\r\nRotate your gradient.\r\n\r\nTips:\r\nWhen the Rotation Animation Type is set to 'Loop' or 'Ping-Pong', the Rotation Angle define the Animation Start Point (from where the animation must start the rotation). \r\n\r\nWhen Radial Gradient is selected, you need to set a Gradient Offset to see the gradient rotation, otherwise, the gradient will look static (rotating on its very center). \r\n"
   group="[3] Colors"
   parameterName=Gradient Rotation Angle
   sliderMax=1
   sliderMin=-1
   sortPriority=14
  }
  NamedRerouteDeclaration(None=@n183) {name=Gradient Rotation Angle, nodeColor=<rgba 0 0 0 1>}
  NamedRerouteDeclaration(None=@n157) {name=Stroke Thickness, nodeColor=<rgba 0 0 0 1>}
  n184 = ScalarParameter() {
   desc="0.0 = Static (default) \r\n1.0 = Loop \r\n2.0 = Ping Pong \r\n3.0 = Switch (A then directly B) \r\n\r\nThe type of the automatic and infinite rotation. \r\n\r\nNeeds a Rotation Speed different of 0 to be activated.\r\n\r\nWarning:\r\nthe rotation of the texture will work inside its rendering container. so rotate may deform a rectangle shape unless you:\r\n\r\n - set the same width and height material size and image/widget size (the material container).\r\n\r\n- Use a different UV Scaling Type to Scale to Fit or Scale to Fill the material container.\r\n"
   group="[3] Colors"
   parameterName=Select Gradient Rotation Animation Type
   sortPriority=15
  }
  NamedRerouteDeclaration(None=@n184) {name="Gradient  Rotation Animation Type", nodeColor=<rgba 0 0 0 1>}
  n185 = ScalarParameter() {
   defaultValue=0.15
   desc="Default Value = 0.15\r\n\r\nDepending if the value is postive or negative, the rotation will be clockwise or counter-clockwise. \r\n\r\nTips:\r\nYou must Select one of the Gradient Rotation Animation Type to see the automatic rotation. \r\n"
   group="[3] Colors"
   parameterName=Gradient Rotation Speed
   sortPriority=17
  }
  NamedRerouteDeclaration(None=@n185) {name=Gradient Rotation Speed, nodeColor=<rgba 0 0 0 1>}
  n186 = MaterialFunctionCall(
   UVs=NamedRerouteUsage()
   Size=NamedRerouteUsage()
   Outline Thickness=NamedRerouteUsage()
   Glow Min=NamedRerouteUsage()
   Glow Max=Constant()
   Corner Roundness=NamedRerouteUsage()
   Keep Aspect Ratio=NamedRerouteUsage()
  ) {materialFunction=&MF_UGCUI_Box}
  n187 = LinearInterpolate(A=@n186.Fill, B=Subtract(A=@n186.Fill, B=NamedRerouteUsage()), Alpha=NamedRerouteUsage())
  n188 = NamedRerouteUsage()
  NamedRerouteDeclaration(None=Saturate(None=Subtract(A=@n187, B=@n188))) {name=Shape Fill, nodeColor=<rgba 0 0 0 1>}
  n189 = ScalarParameter() {
   desc="0.0 = No Ease used (Default) \r\n1.0 =EaseInSine \r\n2.0 = EaseOutSine \r\n3.0 = EaseInOutSine \r\n4.0 = EaseInCubic \r\n5.0 = EaseOutCubic \r\n6.0 = EaseInOutCubic \r\n7.0 = EaseInQuint \r\n8.0 = EaseOutQuint \r\n9.0 = EaseInOutQuint \r\n\r\nThe easing curve of the animation. \r\n"
   group="[3] Colors"
   parameterName=Select Gradient Rotation Easing Style
   sortPriority=16
  }
  NamedRerouteDeclaration(None=@n189) {name="Gradient  Rotation Easing Style", nodeColor=<rgba 0 0 0 1>}
  n190 = ScalarParameter() {
   defaultValue=1
   desc="1.0 = Stretch (default)\r\n2.0 =Scaled to Fit \r\n3.0 = Scaled to Fill \r\n\r\nTips:\r\nIf your image size / Widget size is not a square, you may want to adjust this parameter to keep your shape aspect ratio. \r\n\r\nOptions 2 and 3 keep the aspect ratio of the shape. \r\n\r\nWARNING:\r\n\r\nRotation on streched widget NEEDS a scale to fit or to fill to avoid distortions.\r\n \r\nWhen active, you must disable the 'Keep Corners Aspect Ratio' parameter to avoid cancelling the expected result.\r\n"
   group="[1] Fill"
   parameterName=Select the Scaling Method
   sliderMax=3
   sliderMin=1
   sortPriority=3
  }
  NamedRerouteDeclaration(None=Floor(None=@n190)) {name=Shape UV Scaling, nodeColor=<rgba 0 0 0 1>}
  NamedRerouteDeclaration(None=MaterialFunctionCall(UV=NamedRerouteUsage(), UV Scaling Switch Index=NamedRerouteUsage()) {materialFunction=&MF_UGCUI_UVScalingSwitcher}.Selected Scaled UV) {name=Shape Scaled UV, nodeColor=<rgba 1 0.0036 0 1>}
  n191 = NamedRerouteUsage()
  n192 = NamedRerouteUsage()
  n193 = NamedRerouteUsage()
  n194 = Saturate(None=Multiply(A=Multiply(A=@n191, B=@n192), B=ComponentMask(None=@n193) {a=true, desc="The alpha of the Shape Color \r\nalso drives the shape alpha"}))
  n195 = NamedRerouteUsage()
  n196 = NamedRerouteUsage()
  n197 = NamedRerouteUsage()
  n198 = Saturate(None=Multiply(A=Multiply(A=@n195, B=@n196), B=ComponentMask(None=@n197) {a=true, desc="The alpha of the stroke color \r\nalso drives the stroke alpha"}))
  NamedRerouteDeclaration(None=Max(
    A=NamedRerouteDeclaration(None=@n194) {name=Idle Shape Alpha, nodeColor=<rgba 0 0 0 1>}
    B=NamedRerouteDeclaration(None=@n198) {name=Idle Stroke Alpha, nodeColor=<rgba 0 0 0 1>}
   )) {name=Idle Alpha, nodeColor=<rgba 0 0 0 1>}
  n199 = ScalarParameter() {
   desc="Default value = 0.0 \r\nRange: 0.0-1.0\r\n\r\nSet the transparency between 0 (transparent) and 1 (full opaque).\r\n\r\nInside the Shape Fill, the Stroke may not be visible unless you reduce the Transparency of the shape Fill.\r\n"
   group="[2] Stroke"
   parameterName=Stroke Transparency
   sliderMax=1
   sortPriority=1
  }
  NamedRerouteDeclaration(None=@n199) {name=Idle Stroke Transparency, nodeColor=<rgba 0 0 0 1>}
  n200 = ScalarParameter() {
   defaultValue=0.65
   desc="Default Value = 0.65 (0-1)\r\n\r\nSet the transparency between 0 (transparent) and 1 (full opaque).\r\n"
   group="[4a] State-Hovered"
   parameterName=Hovered Fill Transparency
   sliderMax=1
   sortPriority=5
  }
  NamedRerouteDeclaration(None=@n200) {name=Hovered Fill Transparency, nodeColor=<rgba 0 0 0 1>}
  n201 = ScalarParameter() {
   defaultValue=0.8
   desc="Default Value = 0 (0-1)\r\n\r\nSet the transparency between 0 (transparent) and 1 (full opaque).\r\n"
   group="[4a] State-Hovered"
   parameterName=Hovered Stroke Transparency
   sliderMax=1
   sortPriority=6
  }
  NamedRerouteDeclaration(None=@n201) {name=Hovered Stroke Transparency, nodeColor=<rgba 0 0 0 1>}
  n202 = ScalarParameter() {
   defaultValue=0.65
   desc="Default Value = 0.65 (0-1)\r\n\r\nSet the transparency between 0 (transparent) and 1 (full opaque).\r\n"
   group="[4b] State-Focused"
   parameterName=Focused Fill Transparency
   sliderMax=1
   sortPriority=5
  }
  NamedRerouteDeclaration(None=@n202) {name=Focused Fill Transparency, nodeColor=<rgba 0 0 0 1>}
  n203 = ScalarParameter() {
   defaultValue=0.8
   desc="Default Value = 0 (0-1)\r\n\r\nSet the transparency between 0 (transparent) and 1 (full opaque).\r\n"
   group="[4b] State-Focused"
   parameterName=Focused Stroke Transparency
   sliderMax=1
   sortPriority=6
  }
  NamedRerouteDeclaration(None=@n203) {name=Focused Stroke Transparency, nodeColor=<rgba 0 0 0 1>}
  n204 = ScalarParameter() {
   defaultValue=0.5
   desc="Default Value = 0.5 (0-1)\r\n\r\nSet the transparency between 0 (transparent) and 1 (full opaque).\r\n"
   group="[4c] State-Pressed"
   parameterName=Pressed Fill Transparency
   sliderMax=1
   sortPriority=5
  }
  NamedRerouteDeclaration(None=@n204) {name=Pressed Fill Transparency, nodeColor=<rgba 0 0 0 1>}
  n205 = ScalarParameter() {
   defaultValue=0.5
   desc="Default Value = 0 (0-1)\r\n\r\nSet the transparency between 0 (transparent) and 1 (full opaque).\r\n"
   group="[4c] State-Pressed"
   parameterName=Pressed Stroke Transparency
   sliderMax=1
   sortPriority=6
  }
  NamedRerouteDeclaration(None=@n205) {name=Pressed Stroke Transparency, nodeColor=<rgba 0 0 0 1>}
  n206 = ScalarParameter() {
   defaultValue=0.5
   desc="Default Value = 0.5 (0-1)\r\n\r\nSet the transparency between 0 (transparent) and 1 (full opaque).\r\n"
   group="[4d] State-Disabled"
   parameterName=Disabled Fill Transparency
   sliderMax=1
   sortPriority=5
  }
  NamedRerouteDeclaration(None=@n206) {name=Disabled Fill Transparency, nodeColor=<rgba 0 0 0 1>}
  NamedRerouteUsage()
  NamedRerouteUsage()
 ]
}
