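# Material function for UEFN Modular Button UI materials that blends per-state color and opacity for Idle, Hovered, Focused, Pressed and Disabled states.
# Each state's color and opacity, plus four scalar transition inputs, are saturated to the 0–1 range.
# Two parallel chains of nested lerps start from Idle and apply Hovered, Focused, Pressed, then Disabled, so later states override earlier ones at full transition.
# The chains share transition alphas and feed the Color and Opacity outputs.
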MaterialFunction /FortUILibrary/UGC/MaterialFunctions/ButtonStates/MF_UEFN_UI_ButtonStatesBase_Lerp
graph {
 properties {
  bExposeToLibrary=true
  description=Allows defining visuals for Modular Button Base states inside the material. Controls switching between the states by using transition parameters.
  libraryCategoriesText [UI Modular Button]
 }
 nodes [
  n1 = FunctionInput() {bUsePreviewValueAsDefault=true, description=Color for Idle state, inputName=Idle, previewValue=<1 1 1 1>}
  n2 = FunctionInput() {bUsePreviewValueAsDefault=true, description=Color for Hovered state, inputName=Hovered, sortPriority=10, previewValue=<1 1 1 1>}
  n3 = FunctionInput() {
   bUsePreviewValueAsDefault=true
   description=Transition
   inputName=Trans_Hovered
   inputType=FunctionInput_Scalar
   sortPriority=12
   previewValue=<1 0 0 0>
  }
  n4 = Saturate(None=@n3)
  n5 = FunctionInput() {bUsePreviewValueAsDefault=true, description=Color for Focused state, inputName=Focused, sortPriority=20, previewValue=<1 1 1 1>}
  n6 = FunctionInput() {
   bUsePreviewValueAsDefault=true
   description=Transition
   inputName=Trans_Focused
   inputType=FunctionInput_Scalar
   sortPriority=22
   previewValue=<1 0 0 0>
  }
  n7 = Saturate(None=@n6)
  n8 = LinearInterpolate(A=LinearInterpolate(A=Saturate(None=@n1), B=Saturate(None=@n2), Alpha=@n4), B=Saturate(None=@n5), Alpha=@n7)
  n9 = Saturate(None=FunctionInput() {bUsePreviewValueAsDefault=true, description=Color for Pressed state, inputName=Pressed, sortPriority=30})
  n10 = FunctionInput() {
   bUsePreviewValueAsDefault=true
   description=Transition
   inputName=Trans_Pressed
   inputType=FunctionInput_Scalar
   sortPriority=32
   previewValue=<1 0 0 0>
  }
  n11 = Saturate(None=@n10)
  n12 = FunctionInput() {bUsePreviewValueAsDefault=true, description=Color for Disabled state, inputName=Disabled, sortPriority=40}
  n13 = FunctionInput() {
   bUsePreviewValueAsDefault=true
   description=Transition
   inputName=Trans_Disabled
   inputType=FunctionInput_Scalar
   sortPriority=42
   previewValue=<1 0 0 0>
  }
  n14 = Saturate(None=@n13)
  FunctionOutput(None=LinearInterpolate(A=LinearInterpolate(A=@n8, B=@n9, Alpha=@n11), B=Saturate(None=@n12), Alpha=@n14)) {outputName=Color}
  n15 = FunctionInput() {
   bUsePreviewValueAsDefault=true
   description=Opacity for Idle state
   inputName=Idle Opacity
   inputType=FunctionInput_Scalar
   sortPriority=1
   previewValue=<1 0 0 0>
  }
  n16 = FunctionInput() {
   bUsePreviewValueAsDefault=true
   description=Opacity for Hovered state
   inputName=Hovered Opacity
   inputType=FunctionInput_Scalar
   sortPriority=11
   previewValue=<1 0 0 0>
  }
  n17 = FunctionInput() {
   bUsePreviewValueAsDefault=true
   description=Opacity for Focused state
   inputName=Focused Opacity
   inputType=FunctionInput_Scalar
   sortPriority=21
   previewValue=<1 0 0 0>
  }
  n18 = LinearInterpolate(A=LinearInterpolate(A=Saturate(None=@n15), B=Saturate(None=@n16), Alpha=@n4), B=Saturate(None=@n17), Alpha=@n7)
  n19 = FunctionInput() {
   bUsePreviewValueAsDefault=true
   description=Opacity for Pressed state
   inputName=Pressed Opacity
   inputType=FunctionInput_Scalar
   sortPriority=31
   previewValue=<1 0 0 0>
  }
  n20 = Saturate(None=@n19)
  n21 = FunctionInput() {
   bUsePreviewValueAsDefault=true
   description=Opacity for Disabled state
   inputName=Disabled Opacity
   inputType=FunctionInput_Scalar
   sortPriority=41
   previewValue=<1 0 0 0>
  }
  FunctionOutput(None=LinearInterpolate(A=LinearInterpolate(A=@n18, B=@n20, Alpha=@n11), B=Saturate(None=@n21), Alpha=@n14)) {outputName=Opacity, sortPriority=1}
 ]
}
