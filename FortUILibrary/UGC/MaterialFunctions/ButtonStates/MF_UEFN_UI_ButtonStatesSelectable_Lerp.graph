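# Material function for UEFN modular UI buttons that layers Selected-state color and opacity over incoming Base states.
# Chained lerps apply Idle, Hovered, Focused, Pressed and Disabled in order, so later states take priority as their transition scalars rise.
# Color and opacity chains share those scalar parameters.
# State inputs default to the RGB and alpha of per-state color parameters, so callers can override either.
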MaterialFunction /FortUILibrary/UGC/MaterialFunctions/ButtonStates/MF_UEFN_UI_ButtonStatesSelectable_Lerp
graph {
 properties {
  bExposeToLibrary=true
  description=Allows defining visuals for Modular Button Selected states inside the material. Controls switching between the states by using transition parameters.
  libraryCategoriesText [UI Modular Button]
 }
 nodes [
  n1 = FunctionInput() {
   desc=Selected states added on top of Base of Locked states
   description=Input for Base states color
   inputName=Base States
   previewValue=<1 1 1 1>
  }
  n2 = VectorParameter() {group=State colors, parameterName=Color_SelectedIdle, sortPriority=10, defaultValue=<rgba 1 1 1 1>}
  n3 = FunctionInput(Preview=@n2.RGB) {bUsePreviewValueAsDefault=true, description=Color for Selected_Idle state, inputName=Selected_Idle, sortPriority=2, previewValue=<1 1 1 1>}
  n4 = ScalarParameter() {group=State transitions, parameterName=Trans_SelectedIdle, sliderMax=1, sortPriority=10}
  n5 = VectorParameter() {group=State colors, parameterName=Color_SelectedHovered, sortPriority=11, defaultValue=<rgba 1 1 1 1>}
  n6 = FunctionInput(Preview=@n5.RGB) {
   bUsePreviewValueAsDefault=true
   description=Color for Selected_Hovered state
   inputName=Selected_Hovered
   sortPriority=10
   previewValue=<1 1 1 1>
  }
  n7 = ScalarParameter() {group=State transitions, parameterName=Trans_SelectedHovered, sliderMax=1, sortPriority=11}
  n8 = VectorParameter() {group=State colors, parameterName=Color_SelectedFocused, sortPriority=12, defaultValue=<rgba 1 1 1 1>}
  n9 = FunctionInput(Preview=@n8.RGB) {
   bUsePreviewValueAsDefault=true
   description=Color for Selected_Focused state
   inputName=Selected_Focused
   sortPriority=20
   previewValue=<1 1 1 1>
  }
  n10 = ScalarParameter() {group=State transitions, parameterName=Trans_SelectedFocused, sliderMax=1, sortPriority=12}
  n11 = LinearInterpolate(A=LinearInterpolate(A=LinearInterpolate(A=@n1, B=@n3, Alpha=@n4), B=@n6, Alpha=@n7), B=@n9, Alpha=@n10)
  n12 = VectorParameter() {group=State colors, parameterName=Color_SelectedPressed, sortPriority=13, defaultValue=<rgba 1 1 1 1>}
  n13 = FunctionInput(Preview=@n12.RGB) {bUsePreviewValueAsDefault=true, description=Color for Selected_Pressed state, inputName=Selected_Pressed, sortPriority=40}
  n14 = ScalarParameter() {group=State transitions, parameterName=Trans_SelectedPressed, sliderMax=1, sortPriority=13}
  n15 = VectorParameter() {group=State colors, parameterName=Color_SelectedDisabled, sortPriority=14, defaultValue=<rgba 1 1 1 1>}
  n16 = FunctionInput(Preview=@n15.RGB) {bUsePreviewValueAsDefault=true, description=Color for Selected_Disabled state, inputName=Selected_Disabled, sortPriority=50}
  n17 = ScalarParameter() {group=State transitions, parameterName=Trans_SelectedDisabled, sliderMax=1, sortPriority=14}
  FunctionOutput(None=LinearInterpolate(A=LinearInterpolate(A=@n11, B=@n13, Alpha=@n14), B=@n16, Alpha=@n17)) {outputName=Color}
  n18 = FunctionInput() {description=Input for Base states opacity, inputName=Base States Opacity, inputType=FunctionInput_Scalar, sortPriority=1}
  n19 = FunctionInput(Preview=@n2.A) {
   bUsePreviewValueAsDefault=true
   description=Opacity for Selected_Idle state
   inputName=Selected_Idle Opacity
   inputType=FunctionInput_Scalar
   sortPriority=3
  }
  n20 = FunctionInput(Preview=@n5.A) {
   bUsePreviewValueAsDefault=true
   description=Opacity for Selected_Hovered state
   inputName=Selected_Hovered Opacity
   inputType=FunctionInput_Scalar
   sortPriority=11
  }
  n21 = FunctionInput(Preview=@n8.A) {
   bUsePreviewValueAsDefault=true
   description=Opacity for Selected_Focused state
   inputName=Selected_Focused Opacity
   inputType=FunctionInput_Scalar
   sortPriority=21
  }
  n22 = LinearInterpolate(A=LinearInterpolate(A=LinearInterpolate(A=@n18, B=@n19, Alpha=@n4), B=@n20, Alpha=@n7), B=@n21, Alpha=@n10)
  n23 = FunctionInput(Preview=@n12.A) {
   bUsePreviewValueAsDefault=true
   description=Opacity for Selected_Pressed state
   inputName=Selected_Pressed Opacity
   inputType=FunctionInput_Scalar
   sortPriority=41
  }
  n24 = FunctionInput(Preview=@n15.A) {
   bUsePreviewValueAsDefault=true
   description=Opacity for Selected_Disabled state
   inputName=Selected_Disabled Opacity
   inputType=FunctionInput_Scalar
   sortPriority=51
  }
  FunctionOutput(None=LinearInterpolate(A=LinearInterpolate(A=@n22, B=@n23, Alpha=@n14), B=@n24, Alpha=@n17)) {outputName=Opacity, sortPriority=1}
 ]
}
