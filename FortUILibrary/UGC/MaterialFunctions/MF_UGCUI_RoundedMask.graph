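# Material function for UGC UI that generates a rounded-rectangle shape with per-corner radii, outputting fill and outline SDFs, masks, UVs and a viewport scale factor.
# Radius, given as a ratio or in pixels, is clamped, scaled by the smaller size axis, then fed to MF_UGCUI_Box.
# Smoothstep bounds derived from an edge falloff soften both masks.
# Further nodes compute skewed UVs, corner aliasing masks and thresholds for MF_UGCUI_AA.
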MaterialFunction /FortUILibrary/UGC/MaterialFunctions/MF_UGCUI_RoundedMask
graph {
 properties {bExposeToLibrary=true, description=Generate a rounded corner shape based on Core UI rules, libraryCategoriesText [UI Library]}
 nodes [
  Constant() {r=0.3}
  Constant()
  n1 = NamedRerouteUsage()
  n2 = NamedRerouteUsage()
  n3 = NamedRerouteUsage()
  n4 = FunctionInput(Preview=Constant2Vector() {g=1, r=1}) {bUsePreviewValueAsDefault=true, description=Defines the size, inputName=Size, inputType=FunctionInput_Vector2, sortPriority=1}
  n5 = FunctionInput() {
   bUsePreviewValueAsDefault=true
   description="Set the corners radius. between 0 and 1 if Radius in Pixels is set to off, in number of pixels if Radius in Pixels is set to on."
   inputName=Corner Radius
   inputType=FunctionInput_Vector4
   sortPriority=2
   previewValue=<0.3 0.3 0.3 3>
  }
  n6 = NamedRerouteDeclaration(None=@n5) {name=Corner Radius, nodeColor=<rgba 0 0 0 1>}
  n7 = Divide(A=@n6, B=NamedRerouteUsage())
  n8 = FunctionInput() {
   bUsePreviewValueAsDefault=true
   description="Set if you desire to calculate the corner radius in pixels instead of a ratio/percentage. 0 = Off (default), Greater than 0 = On."
   inputName=Radius in Pixels on/off
   inputType=FunctionInput_Scalar
   sortPriority=3
  }
  n9 = Saturate(None=Ceil(None=@n8))
  n10 = Multiply(A=Multiply(A=LinearInterpolate(A=@n6, B=@n7, Alpha=@n9)) {constB=2}, B=LinearInterpolate(B=NamedRerouteUsage(), Alpha=@n9) {constA=1})
  n11 = NamedRerouteUsage()
  n12 = ComponentMask(None=@n11) {r=true}
  n13 = ComponentMask(None=@n11) {g=true}
  n14 = Multiply(
   A=NamedRerouteDeclaration(None=Min(A=@n10) {desc=Avoid rounded corners to exceed the maximum amount}) {name=Remapped Corner Radius, nodeColor=<rgba 0 0 0 1>}
   B=NamedRerouteDeclaration(None=Min(A=@n12, B=@n13)) {name=Min Size, nodeColor=<rgba 0 0 0 1>}
  )
  n15 = FunctionOutput(None=MaterialFunctionCall!(
    UVs=@n3
    Size=NamedRerouteDeclaration(None=@n4) {name=Size, nodeColor=<rgba 0 0 0 1>}
    Outline Thickness
    Glow Min
    Glow Max
    Corner Roundness=NamedRerouteDeclaration(None=@n14) {name=Corner Radius with Padding, nodeColor=<rgba 0 0 0 1>}
    Keep Aspect Ratio=Constant() {r=1}
   ) {materialFunction=&/MF_UGCUI_Box}.SDF) {description=Output the SDF, outputName=SDF, sortPriority=10}
  n16 = SmoothStep(Min=@n1, Max=@n2, Value=@n15)
  FunctionOutput(None=OneMinus(None=Saturate(None=@n16))) {description=Output the mask, outputName=Shape Mask}
  FunctionOutput(None=@n3) {description=Output the UVs, outputName=UVs, sortPriority=50}
  n17 = FunctionInput(Preview=Constant() {r=0.2}) {
   bUsePreviewValueAsDefault=true
   description=Define the thickness of the outline
   inputName=Outline Thickness
   inputType=FunctionInput_Scalar
   sortPriority=20
  }
  n18 = Multiply(A=Divide(A=@n17, B=NamedRerouteUsage()), B=NamedRerouteUsage())
  n19 = Add(A=@n15, B=@n18)
  n20 = FunctionOutput(None=Subtract(A=Abs(None=@n19), B=@n18) {constB=0.1}) {description=Output the outline SDF, outputName=Outline SDF, sortPriority=11}
  n21 = SmoothStep(Min=@n1, Max=@n2, Value=@n20)
  FunctionOutput(None=OneMinus(None=Saturate(None=@n21))) {description=Output the outline mask, outputName=Outline Mask, sortPriority=1}
  n22 = ViewSize()
  n23 = ComponentMask(None=@n22) {r=true}
  n24 = ComponentMask(None=@n22) {g=true}
  n25 = MaterialFunctionCall!(A=@n23, B=@n24) {materialFunction=&/MF_UGCUI_BooleanOperations}
  n26 = LinearInterpolate(A=@n23, B=@n24, Alpha=@n25.IsGreater)
  n27 = Constant2Vector() {g=1080, r=1920}
  n28 = LinearInterpolate(A=@n27.R, B=@n27.R, Alpha=@n25.IsGreater)
  FunctionOutput(None=NamedRerouteDeclaration(None=Divide(A=@n26, B=@n28)) {name=Resolution Factor, nodeColor=<rgba 0 0 0 1>}) {
   description=Output the viewport scale factor that adjust the rounded corners based on screen dimensions
   outputName=Viewport Scale Factor
   sortPriority=70
  }
  n29 = NamedRerouteUsage()
  LinearInterpolate(A=NamedRerouteUsage(), Alpha=@n29)
  n30 = Multiply(A=Divide(A=Constant() {desc=Aliasing Adjustment Value, r=1.5}, B=NamedRerouteUsage()), B=NamedRerouteUsage()) {desc=Offset by 1 normalized pixel}
  n31 = Max(A=Saturate(None=Add(A=@n30) {constB=0.5})) {constB=0.5}
  n32 = NamedRerouteUsage()
  n33 = Max(A=LinearInterpolate(B=LinearInterpolate(A=@n31, Alpha=@n32) {constB=0.5}, Alpha=NamedRerouteUsage()) {constA=1, constAlpha=0, constB=0.5, desc=Adjust the threshold based on the corners}) {constB=0}
  MaterialFunctionCall!(Threshold=@n33, DistanceField=Add(A=@n15) {constB=0.5}) {materialFunction=&/MF_UGCUI_AA}
  MaterialFunctionCall(Threshold=@n33, DistanceField=Add(A=@n20) {constB=0.5}) {materialFunction=&MF_UGCUI_AA}
  n34 = NamedRerouteUsage()
  n35 = ComponentMask(None=@n34) {r=true}
  n36 = ComponentMask(None=@n34) {g=true}
  NamedRerouteDeclaration(None=Max(A=Min(A=@n35, B=@n36)) {constB=0}) {name=Min Pixel Size, nodeColor=<rgba 0 0 0 1>}
  n37 = FunctionInput(Preview=Constant()) {bUsePreviewValueAsDefault=true, description=Defines the skew angle, inputName=Skew Angle, inputType=FunctionInput_Scalar, sortPriority=5}
  n38 = Multiply(
   A=ComponentMask(None=MaterialFunctionCall() {materialFunction=&/MF_UGCUI_GetAspectRatio}.V Ratio) {g=true}
   B=NamedRerouteDeclaration(None=@n37) {name=Skew Angle, nodeColor=<rgba 0 0 0 1>}
  )
  n39 = Constant()
  NamedRerouteDeclaration(None=MaterialFunctionCall!(UV=NamedRerouteUsage(), Skew Angle=AppendVector(A=@n38, B=@n39)) {materialFunction=&/MF_UGCUI_Skew}.Skewed UVs) {name=Skewed UVs, nodeColor=<rgba 0 0 0 1>}
  n40 = NamedRerouteUsage()
  n41 = ComponentMask(None=@n40) {desc=Top Left, r=true}
  n42 = ComponentMask(None=@n40) {b=true, desc=Top Right}
  n43 = Round(None=NamedRerouteUsage())
  n44 = ComponentMask(None=@n43) {g=true}
  n45 = OneMinus(None=@n44)
  n46 = Abs(None=Subtract(A=NamedRerouteUsage()) {constB=0.5})
  n47 = Multiply(A=Divide(A=Constant() {r=1.5}, B=NamedRerouteUsage()), B=NamedRerouteUsage())
  n48 = Saturate(None=Subtract(A=Multiply(A=@n46) {constB=2}, B=OneMinus(None=@n47)))
  n49 = ComponentMask(None=@n48) {g=true}
  n50 = ComponentMask(None=@n40) {a=true, desc=Bottom Right}
  n51 = ComponentMask(None=@n43) {r=true}
  n52 = ComponentMask(None=@n48) {r=true}
  n53 = Max(A=Multiply(A=Max(A=@n41, B=@n42), B=Multiply(A=@n45, B=@n49)), B=Multiply(A=Max(A=@n42, B=@n50), B=Multiply(A=@n51, B=@n52)))
  n54 = ComponentMask(None=@n40) {desc=Bottom Left, g=true}
  n55 = OneMinus(None=@n51)
  n56 = Max(A=Multiply(A=Max(A=@n50, B=@n54), B=Multiply(A=@n44, B=@n49)), B=Multiply(A=Max(A=@n41, B=@n54), B=Multiply(A=@n55, B=@n52)))
  NamedRerouteDeclaration(None=Ceil(None=Max(A=@n53, B=@n56))) {name=Aliasing Edge Mask, nodeColor=<rgba 0 0 0 1>}
  n57 = NamedRerouteUsage()
  n58 = NamedRerouteUsage()
  n59 = Multiply(A=@n58, B=Constant2Vector() {g=0.5, r=0.5})
  n60 = Multiply(A=@n58, B=Constant2Vector() {g=0.5, r=-0.5})
  n61 = MaterialFunctionCall!(UV=NamedRerouteUsage()) {materialFunction=&/MF_UGCUI_CornerMask}
  n62 = LinearInterpolate(
   A=NamedRerouteDeclaration(None=Subtract(A=@n57, B=@n59)) {name=Bottom Right UV, nodeColor=<rgba 0 0 0 1>}
   B=NamedRerouteDeclaration(None=Subtract(A=@n57, B=@n60)) {name=Bottom Left UV, nodeColor=<rgba 0 0 0 1>}
   Alpha=@n61.Bottom Left Mask
  ) {constA=0.5}
  n63 = Constant2Vector() {g=-0.5, r=0.5}
  n64 = NamedRerouteDeclaration(None=Subtract(A=@n57, B=Multiply(A=@n63, B=@n58))) {name=Top Right UV, nodeColor=<rgba 0 0 0 1>}
  n65 = Subtract(A=@n57, B=Multiply(A=Constant2Vector() {g=-0.5, r=-0.5}, B=@n58))
  n66 = NamedRerouteDeclaration(None=LinearInterpolate(
    A=LinearInterpolate(A=@n62, B=@n64, Alpha=@n61.Top Right Mask) {constA=0.5}
    B=NamedRerouteDeclaration(None=@n65) {name=Top Left UV, nodeColor=<rgba 0 0 0 1>}
    Alpha=@n61.Top Left Mask
   ) {constA=0.25}) {name=Combined Corners UV, nodeColor=<rgba 0 0 0 1>}
  n67 = NamedRerouteUsage()
  n68 = ComponentMask(None=@n67) {a=true}
  n69 = ComponentMask(None=@n67) {g=true}
  n70 = ComponentMask(None=@n67) {b=true}
  n71 = LinearInterpolate(A=LinearInterpolate(A=@n68, B=@n69, Alpha=@n61.Bottom Left Mask) {constA=0.5}, B=@n70, Alpha=@n61.Top Right Mask) {constA=0.5}
  n72 = ComponentMask(None=@n67) {r=true}
  n73 = Saturate(None=MaterialFunctionCall!(
    UV=MaterialFunctionCall!(UV=@n66, TextureSize, ScalePivotPoint) {materialFunction=&/MF_UGCUI_UVScaling}.Scaled to Fit
    Size=LinearInterpolate(A=@n71, B=@n72, Alpha=@n61.Top Left Mask) {constA=0.25}
    GlowMin
    GlowMax
    OutlineThickness
   ) {materialFunction=&/MF_UGCUI_Circle}.SDF)
  NamedRerouteDeclaration(None=OneMinus(None=Ceil(None=@n73))) {name=Corner Aliasing Mask, nodeColor=<rgba 0 0 0 1>}
  n74 = Multiply(A=@n72, B=@n70)
  n75 = Multiply(A=@n68, B=@n69)
  NamedRerouteDeclaration(None=Floor(None=Multiply(A=@n74, B=@n75))) {name=Is Circle, nodeColor=<rgba 0 0 0 1>}
  n76 = MaterialFunctionCall() {materialFunction=&GetUserInterfaceUV}
  NamedRerouteDeclaration(None=FunctionInput(Preview=@n76.Normalized UV) {bUsePreviewValueAsDefault=true, description=Define the UVs, inputName=UV, inputType=FunctionInput_Vector2}) {name=UV, nodeColor=<rgba 0 0 0 1>}
  n77 = NamedRerouteDeclaration(None=Constant() {r=0.0075}) {name=Edge Falloff Value, nodeColor=<rgba 0 0 0 1>}
  n78 = Divide(B=NamedRerouteUsage()) {constA=192, constB=1}
  n79 = Max(A=Multiply(A=Multiply(A=@n77, B=@n78), B=NamedRerouteUsage())) {constB=0}
  NamedRerouteDeclaration(None=Multiply(A=NamedRerouteDeclaration(None=@n79) {name=Max Value, nodeColor=<rgba 0 0 0 1>}) {constB=-1}) {name=Min Value, nodeColor=<rgba 0 0 0 1>}
  NamedRerouteDeclaration(None=@n76.Pixel Size) {name=Pixel Size, nodeColor=<rgba 0 0 0 1>}
  n80 = NamedRerouteUsage()
  NamedRerouteDeclaration(None=LinearInterpolate(B=@n29, Alpha=Floor(None=@n80)) {constA=1}) {name=Aliasing Mask, nodeColor=<rgba 0 0 0 1>}
 ]
}
