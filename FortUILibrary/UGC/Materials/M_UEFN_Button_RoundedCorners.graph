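# UEFN UI material for rounded-corner buttons, shown as either a fill or an outline.
# Button-state functions lerp colors and opacities by transition parameters, with optional selectable states (selected inputs unwired) and a rotatable two-color gradient override.
# A rounded-mask function takes size, radius (ratio or pixels) and thickness, clipping state opacity through Min.
# Alpha-composite blending uses mask-premultiplied color, with separate brightness and additive scalars tuning color against alpha.
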Material /FortUILibrary/UGC/Materials/M_UEFN_Button_RoundedCorners
graph {
 properties {
  bUsedWithStaticMesh=false
  blendMode=BLEND_AlphaComposite
  materialDomain=MD_UI
  pixelDepthOffsetMode=PDOM_Legacy
  usedShadingModels=Default Lit
 }
 nodes [
  n1 = ScalarParameter() {
   desc="0 = Off (default)\r\nGreater than 0 = On\r\n\r\nWhen enabled, selected state colors, transition and opacity are used.\r\n"
   group=Global Settings
   parameterName=Is Selectable on/off
   sliderMax=1
   sortPriority=0
  }
  n2 = Saturate(None=Ceil(None=@n1))
  n3 = VectorParameter() {group=State colors, parameterName=Color_Idle, sortPriority=0, defaultValue=<rgba 0.6297 0 1 1>}
  n4 = VectorParameter() {group=State colors, parameterName=Color_Hovered, sortPriority=1, defaultValue=<rgba 1 1 1 1>}
  n5 = VectorParameter() {group=State colors, parameterName=Color_Focused, sortPriority=2, defaultValue=<rgba 1 1 1 1>}
  n6 = VectorParameter() {group=State colors, parameterName=Color_Pressed, sortPriority=3, defaultValue=<rgba 1 1 1 1>}
  n7 = VectorParameter() {group=State colors, parameterName=Color_Disabled, sortPriority=4, defaultValue=<rgba 1 1 1 1>}
  n8 = MaterialFunctionCall!(
   Idle=@n3.RGB
   Idle Opacity=@n3.RGB
   Hovered=@n4.RGB
   Hovered Opacity=@n4.RGB
   Trans_Hovered=ScalarParameter() {group=State transitions, parameterName=Trans_Hovered, sliderMax=1, sortPriority=1}
   Focused=@n5.RGB
   Focused Opacity=@n5.RGB
   Trans_Focused=ScalarParameter() {group=State transitions, parameterName=Trans_Focused, sliderMax=1, sortPriority=2}
   Pressed=@n6.RGB
   Pressed Opacity=@n6.RGB
   Trans_Pressed=ScalarParameter() {group=State transitions, parameterName=Trans_Pressed, sliderMax=1, sortPriority=3}
   Disabled=@n7.RGB
   Disabled Opacity=@n7.RGB
   Trans_Disabled=ScalarParameter() {group=State transitions, parameterName=Trans_Disabled, sliderMax=1, sortPriority=4}
  ) {materialFunction=/FortUILibrary/UGC/MaterialFunctions/ButtonStates/MF_UEFN_UI_ButtonStatesBase_Lerp}
  n9 = MaterialFunctionCall!(
   Base States=@n8.Color
   Base States Opacity=@n8.Color
   Selected_Idle
   Selected_Idle Opacity
   Selected_Hovered
   Selected_Hovered Opacity
   Selected_Focused
   Selected_Focused Opacity
   Selected_Pressed
   Selected_Pressed Opacity
   Selected_Disabled
   Selected_Disabled Opacity
  ) {materialFunction=/FortUILibrary/UGC/MaterialFunctions/ButtonStates/MF_UEFN_UI_ButtonStatesSelectable_Lerp}
  n10 = ScalarParameter() {
   desc="0 = Off (default)\r\nGreater than 0 = On\r\n\r\nWhen enabled, Color A and Color B are used to make the gradient. \r\nWhen disabled, Colors from button states are used. \r\n\r\nThis means that gradient colors A and B are applied to each button state and override the state colors.\r\n"
   group=Gradient
   parameterName=Gradient on/off
   sliderMax=1
   sortPriority=0
  }
  n11 = LinearInterpolate(A=LinearInterpolate(A=@n8.Color, B=@n9.Color, Alpha=@n2) {desc=Colors}, B=NamedRerouteUsage(), Alpha=Saturate(None=Ceil(None=@n10)))
  n12 = ScalarParameter() {
   desc="0 = only the Fill is displayed (default)\r\nGreater than 0 = only the Outline is displayed\r\n"
   group=Global Shape Settings
   parameterName=Display Fill or Oultine
   sortPriority=0
  }
  n13 = Ceil(None=@n12)
  n14 = Min(
   A=LinearInterpolate(A=@n8.Opacity, B=@n9.Opacity, Alpha=@n2) {desc=Opacity}
   B=LinearInterpolate(A=NamedRerouteUsage(), B=NamedRerouteUsage(), Alpha=Saturate(None=@n13))
  )
  n15 = Multiply(A=ScalarParameter() {defaultValue=1, group=Global Settings, parameterName=Brightness Percentage, sliderMax=1, sortPriority=2}, B=@n14)
  Multiply(
   A=ScalarParameter() {defaultValue=1, group=Global Settings, parameterName=Opacity, sliderMax=1}
   B=Multiply(A=VertexColor().A, B=Multiply(A=@n11, B=@n15))
  )
  n16 = ScalarParameter() {
   defaultValue=0.3
   desc="Set the corners radius. \r\n\r\nBetween 0 and 1 if Radius in Pixels is set to: off\r\nIn number of pixels if Radius in Pixels is set to: on\r\n"
   group=Global Shape Settings
   parameterName=Corner Radius
   sortPriority=3
  }
  n17 = ScalarParameter() {
   desc="0 = Off (default) \r\nGreater than 0 = On\r\n\r\nSet if you desire to calculate the corner radius in pixels instead of a ratio/percentage.\r\n"
   group=Global Shape Settings
   parameterName=Radius in Pixels on/off
   sliderMax=1
   sortPriority=4
  }
  n18 = ScalarParameter() {
   defaultValue=1
   desc="Only available if the parameter Display Fill or Outline is set on Outline.\r\n"
   group=Global Shape Settings
   parameterName=Outline Thickness
   sortPriority=5
  }
  n19 = MaterialFunctionCall!(
   UV=NamedRerouteUsage()
   Size=AppendVector(
    A=ScalarParameter() {defaultValue=1, group=Global Shape Settings, parameterName=Length, sortPriority=1}
    B=ScalarParameter() {defaultValue=1, group=Global Shape Settings, parameterName=Height, sortPriority=2}
   )
   Corner Radius=@n16
   Radius in Pixels on/off=@n17
   Skew Angle
   Outline Thickness=@n18
  ) {materialFunction=/FortUILibrary/UGC/MaterialFunctions/MF_UGCUI_RoundedMask}
  NamedRerouteDeclaration(None=@n19.Shape Mask) {name=Fill_Box, nodeColor=<rgba 0 0 0 1>}
  NamedRerouteDeclaration(None=MaterialFunctionCall() {materialFunction=&GetUserInterfaceUV}.Normalized UV) {name=Normalized UV, nodeColor=<rgba 0 0 0 1>}
  n20 = ScalarParameter() {defaultValue=0.5, group=Gradient, parameterName=Gradient Center X, sliderMax=1, sortPriority=4}
  n21 = ScalarParameter() {defaultValue=0.5, group=Gradient, parameterName=Gradient Center Y, sliderMax=1, sortPriority=5}
  n22 = ComponentMask(None=MaterialFunctionCall(
    UVs=TextureCoordinate()
    Rotation Center=AppendVector(A=@n20, B=@n21)
    "Rotation Angle (0-1)"=ScalarParameter() {group=Gradient, parameterName=Gradient Rotation, sliderMax=1, sortPriority=6}
   ) {materialFunction=&CustomRotator}.Rotated Values) {r=true}
  n23 = ScalarParameter() {defaultValue=1, group=Gradient, parameterName=Gradient Softness, sortPriority=3}
  NamedRerouteDeclaration(None=LinearInterpolate(
    A=VectorParameter() {group=Gradient, parameterName=Color A, sortPriority=1, defaultValue=<rgba 0.0646 0.0091 0.125 1>}.RGB
    B=VectorParameter() {group=Gradient, parameterName=Color B, sortPriority=2, defaultValue=<rgba 0.0139 0.1458 0.0685 1>}.RGB
    Alpha=Power(Base=@n22, Exp=@n23)
   )) {name=Gradient Color, nodeColor=<rgba 0 0 0 1>}
  NamedRerouteDeclaration(None=@n19.Outline Mask) {name=Outline_Box, nodeColor=<rgba 0 1 0.1094 1>}
  Reroute(None=Multiply(A=@n14, B=ScalarParameter() {defaultValue=1, group=Global Settings, parameterName=Additive Percentage, sliderMax=1, sortPriority=3}))
 ]
}
