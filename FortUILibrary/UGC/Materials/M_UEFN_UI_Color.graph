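# M_UEFN_UI_Color is a translucent UI-domain material for solid-color widgets that change color and opacity by button state.
# Color and opacity parameters for idle, hovered, focused, pressed and disabled feed the MF_UEFN_UI_ButtonStatesBase_Lerp function along with four transition scalars.
# Parameter descriptions warn that widget transitions depend on the transition names, including TransDisabled, which lacks the underscore used by the other three.
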Material /FortUILibrary/UGC/Materials/M_UEFN_UI_Color
graph {
 properties {
  bUsedWithStaticMesh=false
  blendMode=BLEND_Translucent
  materialDomain=MD_UI
  pixelDepthOffsetMode=PDOM_Legacy
  usedShadingModels=Default Lit
 }
 nodes [
  n1 = VectorParameter() {group="[2a] State-Hovered", parameterName=Hovered Color, sortPriority=0, defaultValue=<rgba 1 1 1 1>}
  n2 = ScalarParameter() {
   desc="WARNING : \r\nThe name of this parameter must not be edited. It is used in many widget parenting. Changing its name will brake the transition behavior.\r\n"
   group="[2a] State-Hovered"
   parameterName=Trans_Hovered
   sliderMax=1
   sortPriority=2
  }
  n3 = VectorParameter() {group="[2b] State-Focused", parameterName=Focused Color, sortPriority=0, defaultValue=<rgba 1 1 1 1>}
  n4 = ScalarParameter() {
   desc="WARNING : \r\nThe name of this parameter must not be edited. It is used in many widget parenting. Changing its name will brake the transition behavior.\r\n"
   group="[2b] State-Focused"
   parameterName=Trans_Focused
   sliderMax=1
   sortPriority=2
  }
  n5 = VectorParameter() {group="[2c] State-Pressed", parameterName=Pressed Color, sortPriority=0, defaultValue=<rgba 1 1 1 1>}
  n6 = ScalarParameter() {
   desc="WARNING : \r\nThe name of this parameter must not be edited. It is used in many widget parenting. Changing its name will brake the transition behavior.\r\n"
   group="[2c] State-Pressed"
   parameterName=Trans_Pressed
   sliderMax=1
   sortPriority=2
  }
  n7 = VectorParameter() {group="[2d] State-Disabled", parameterName=Disabled Color, sortPriority=0, defaultValue=<rgba 1 1 1 1>}
  n8 = ScalarParameter() {
   desc="WARNING : \r\nThe name of this parameter must not be edited. It is used in many widget parenting. Changing its name will brake the transition behavior.\r\n"
   group="[2d] State-Disabled"
   parameterName=TransDisabled
   sliderMax=1
   sortPriority=2
  }
  MaterialFunctionCall!(
   Idle=VectorParameter() {group="[1] Color Settings", parameterName=Color, sortPriority=0, defaultValue=<rgba 1 1 1 1>}.RGB
   Idle Opacity=ScalarParameter() {defaultValue=1, group="[1] Color Settings", parameterName=Opacity, sliderMax=1, sortPriority=1}
   Hovered=@n1.RGB
   Hovered Opacity=ScalarParameter() {defaultValue=1, group="[2a] State-Hovered", parameterName=Hovered Opacity, sliderMax=1, sortPriority=1}
   Trans_Hovered=@n2
   Focused=@n3.RGB
   Focused Opacity=ScalarParameter() {defaultValue=1, group="[2b] State-Focused", parameterName=Focused Opacity, sliderMax=1, sortPriority=1}
   Trans_Focused=@n4
   Pressed=@n5.RGB
   Pressed Opacity=ScalarParameter() {defaultValue=1, group="[2c] State-Pressed", parameterName=Pressed Opacity, sliderMax=1, sortPriority=1}
   Trans_Pressed=@n6
   Disabled=@n7.RGB
   Disabled Opacity=ScalarParameter() {defaultValue=1, group="[2d] State-Disabled", parameterName=Disabled Opacity, sliderMax=1, sortPriority=1}
   Trans_Disabled=@n8
  ) {materialFunction=/FortUILibrary/UGC/MaterialFunctions/ButtonStates/MF_UEFN_UI_ButtonStatesBase_Lerp}
 ]
}
