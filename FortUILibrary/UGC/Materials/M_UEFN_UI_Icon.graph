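# UI material for UEFN widget icons, tinting a texture supplied by the widget Blueprint according to button state.
# A shared lerp function blends Idle, Hovered, Focused, Pressed and Disabled color and opacity through transition scalars.
# Texture alpha passes through an edge-pixel clip, then multiplies with state opacity and a red/blue channel maximum (green is unused) to form a white-mask alpha.
# A floored, saturated "Is RGBA Icon" scalar provides a binary icon-type switch.
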Material /FortUILibrary/UGC/Materials/M_UEFN_UI_Icon
graph {
 properties {
  bUsedWithStaticMesh=false
  blendMode=BLEND_Translucent
  materialDomain=MD_UI
  pixelDepthOffsetMode=PDOM_Legacy
  usedShadingModels=Default Lit
 }
 nodes [
  LinearInterpolate(A=NamedRerouteUsage(), B=NamedRerouteUsage(), Alpha=NamedRerouteUsage())
  LinearInterpolate(A=NamedRerouteUsage(), B=NamedRerouteUsage(), Alpha=NamedRerouteUsage())
  MaterialFunctionCall() {materialFunction=&MakeFloat3}
  n1 = ScalarParameter() {
   defaultValue=1
   desc="Editing this parameter will force the setting and override the blueprint setting.\r\n"
   group="[0] Do Not Edit"
   parameterName=Is RGBA Icon
   sliderMax=1
   sortPriority=0
  }
  NamedRerouteDeclaration(None=Saturate(None=Floor(None=@n1))) {name=Is RGBA Icon, nodeColor=<rgba 0 0 0 1>}
  n2 = TextureSampleParameter2D() {
   desc="DO NOT EDIT : This texture parameter is exposed to let the widget blueprint communicate with the material, and apply the Icon Texture set in the widget blueprint. \r\n\r\nEditing this parameter will force the setting and override the blueprint setting.\r\n\r\n"
   group="[0] Do Not Edit"
   parameterName=Dynamic Icon Texture
   sortPriority=0
  }
  NamedRerouteDeclaration(None=@n2.RGB) {name=Icon RGB, nodeColor=<rgba 1 0 0.4922 1>}
  n3 = VectorParameter() {group="[2a] State-Hovered", parameterName=Hovered Color, sortPriority=0, defaultValue=<rgba 1 1 1 1>}
  n4 = ScalarParameter() {
   desc="WARNING : \r\nThe name of this parameter must not be edited. It is used in many widget parenting. Changing its name will brake the transition behavior.\r\n"
   group="[2a] State-Hovered"
   parameterName=Trans_Hovered
   sliderMax=1
   sortPriority=2
  }
  n5 = VectorParameter() {group="[2b] State-Focused", parameterName=Focused Color, sortPriority=0, defaultValue=<rgba 1 1 1 1>}
  n6 = ScalarParameter() {
   desc="WARNING : \r\nThe name of this parameter must not be edited. It is used in many widget parenting. Changing its name will brake the transition behavior.\r\n"
   group="[2b] State-Focused"
   parameterName=Trans_Focused
   sliderMax=1
   sortPriority=2
  }
  n7 = VectorParameter() {group="[2c] State-Pressed", parameterName=Pressed Color, sortPriority=0, defaultValue=<rgba 1 1 1 1>}
  n8 = ScalarParameter() {
   desc="WARNING : \r\nThe name of this parameter must not be edited. It is used in many widget parenting. Changing its name will brake the transition behavior.\r\n"
   group="[2c] State-Pressed"
   parameterName=Trans_Pressed
   sliderMax=1
   sortPriority=2
  }
  n9 = VectorParameter() {group="[2d] State-Disabled", parameterName=Disabled Color, sortPriority=0, defaultValue=<rgba 1 1 1 1>}
  n10 = ScalarParameter() {
   desc="WARNING : \r\nThe name of this parameter must not be edited. It is used in many widget parenting. Changing its name will brake the transition behavior.\r\n"
   group="[2d] State-Disabled"
   parameterName=TransDisabled
   sliderMax=1
   sortPriority=2
  }
  n11 = MaterialFunctionCall!(
   Idle=VectorParameter() {group="[1] Icon Settings", parameterName=Color, sortPriority=0, defaultValue=<rgba 1 1 1 1>}.RGB
   Idle Opacity=ScalarParameter() {defaultValue=1, group="[1] Icon Settings", parameterName=Opacity, sliderMax=1, sortPriority=1}
   Hovered=@n3.RGB
   Hovered Opacity=ScalarParameter() {defaultValue=1, group="[2a] State-Hovered", parameterName=Hovered Opacity, sliderMax=1, sortPriority=1}
   Trans_Hovered=@n4
   Focused=@n5.RGB
   Focused Opacity=ScalarParameter() {defaultValue=1, group="[2b] State-Focused", parameterName=Focused Opacity, sliderMax=1, sortPriority=1}
   Trans_Focused=@n6
   Pressed=@n7.RGB
   Pressed Opacity=ScalarParameter() {defaultValue=1, group="[2c] State-Pressed", parameterName=Pressed Opacity, sliderMax=1, sortPriority=1}
   Trans_Pressed=@n8
   Disabled=@n9.RGB
   Disabled Opacity=ScalarParameter() {defaultValue=1, group="[2d] State-Disabled", parameterName=Disabled Opacity, sliderMax=1, sortPriority=1}
   Trans_Disabled=@n10
  ) {materialFunction=/FortUILibrary/UGC/MaterialFunctions/ButtonStates/MF_UEFN_UI_ButtonStatesBase_Lerp}
  NamedRerouteDeclaration(None=@n11.Color) {name=Icon Colored by States, nodeColor=<rgba 1 0 0.4922 1>}
  n12 = Max(A=Max(A=@n2.R, B=@n2.R), B=@n2.B)
  n13 = MaterialFunctionCall!(UVs=TextureCoordinate(), Alpha=@n2.A) {materialFunction=/FortUILibrary/UGC/MaterialFunctions/MF_UGCUI_ClipEdgePixel}
  NamedRerouteDeclaration(None=Multiply(A=Multiply(A=@n11.Opacity, B=@n12), B=NamedRerouteDeclaration(None=@n13.Result) {name=Icon Alpha, nodeColor=<rgba 0 0.64 1 1>})) {name=Icon Max White Alpha, nodeColor=<rgba 0 0.64 1 1>}
 ]
}
